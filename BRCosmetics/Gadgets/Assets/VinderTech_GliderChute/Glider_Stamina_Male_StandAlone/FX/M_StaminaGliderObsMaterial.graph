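# Masked unlit Niagara sprite material that draws colored orbs for the stamina glider chute's FX.
# A dynamic "Ball" parameter selects the frame from a 3x3 orb flipbook, sampled at a fixed mip level.
# Particle color is multiplied by the exponent-adjusted texture, and "Texture Amount" blends between flat particle color and that result.
# The blend is divided by eye adaptation to offset exposure, then fed into MF_FX_EmissiveAdjustments.
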Material /BRCosmetics/Gadgets/Assets/VinderTech_GliderChute/Glider_Stamina_Male_StandAlone/FX/M_StaminaGliderObsMaterial
graph {
 properties {bUsedWithNiagaraSprites=true, blendMode=BLEND_Masked, pixelDepthOffsetMode=PDOM_Legacy, shadingModel=MSM_Unlit, usedShadingModels=Unlit}
 nodes [
  n1 = ParticleColor()
  n2 = DynamicParameter() {paramNames [Ball, Texture Amount, Param3, Param4]}
  n3 = TextureSample!(
   UVs=MaterialFunctionCall("Animation  Phase (0-1)"=@n2.Ball, Number of Rows=Constant() {r=3}, Number of Columns=Constant() {r=3}) {materialFunction=&FlipBook}.UVs
   Tex
   Level=Constant()
   Apply View MipBias
  ) {
   mipValueMode=TMVM_MipLevel
   texture=&/T_StaminaStandalone_Orbs
  }
  n4 = Multiply(A=@n1.RGB, B=Power(Base=@n3.RGB, Exp=ScalarParameter() {defaultValue=1, parameterName=Colro Exp}) {constExponent=1.5})
  MaterialFunctionCall(Emissive=Divide(A=LinearInterpolate(A=@n1.RGB, B=@n4, Alpha=@n2.Texture Amount), B=EyeAdaptation().EyeAdaptation)) {materialFunction=&MF_FX_EmissiveAdjustments}
 ]
}
