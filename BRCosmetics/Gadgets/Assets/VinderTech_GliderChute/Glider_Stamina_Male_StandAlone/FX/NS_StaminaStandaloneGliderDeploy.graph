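# Niagara deploy effect for the Stamina standalone glider, layering a delayed glow outline over the skeletal mesh with a burst of flares and shockwaves.
# Ribbons follow spine and claw bones, while sprites cover head sockets, wrists and feet, all tinted by a system outline color.
# Corona, flare with light, curvy shockwave and arc emitters are timed from a shared global delay.
# Front-end modifiers and time-of-day scalars adjust some sizes, rates and brightness.
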NiagaraSystem /BRCosmetics/Gadgets/Assets/VinderTech_GliderChute/Glider_Stamina_Male_StandAlone/FX/NS_StaminaStandaloneGliderDeploy
dependencies {
 usedDataInterfaces [
  &NiagaraDataInterfaceCurve
  &NiagaraDataInterfaceCamera
  &NiagaraDataInterfaceArrayInt32
  &NiagaraDataInterfaceArrayFloat
  &NiagaraDataInterfacePlatformSet
 ]
 usedDynamicInputs []
 usedModules [
  &self:SystemSpawnScript.SystemScriptSource.SystemScriptGraph.NiagaraNodeAssignment_1.SetVariables_#1
  &self:SystemSpawnScript.SystemScriptSource.SystemScriptGraph.NiagaraNodeAssignment_0.SetVariables_#2
  &self:SystemSpawnScript.SystemScriptSource.SystemScriptGraph.NiagaraNodeAssignment_2.SetVariables_#3
  &SystemState
  &self:SystemSpawnScript.SystemScriptSource.SystemScriptGraph.NiagaraNodeAssignment_3.SetVariables_#4
 ]
}
emitters {
 Body_Glow {
  emitterUpdateScript [
   &EmitterState(
    Inactive Response=::NewEnumerator1
    Life Cycle Mode=::NewEnumerator1
    Loop Behavior=::NewEnumerator1
    Loop Delay=0.95
    Loop Duration=10
    UseLoopDelay=true
   )
   /SetVariables(Emitter.num ribbon particles=46) {isSetParameters=true}
   &SpawnBurst_Instantaneous(Spawn Count=$Emitter.num ribbon particles, Use Spawn Probability=false, UseLoopCountLimit=false)
  ]
  particleSpawnScript [
   &InitializeRibbon(Color=$System.Outline Color, Lifetime=1.5, Ribbon Width=600, UsePositionOffset=false, Write Material Random=true)
   &PartitionParticles(
    Number Of Partitions=1
    Particles Per Partition:NiagaraInt32=$Emitter.num ribbon particles
    Partition Mode:ENiagara_PartitionMode=::NewEnumerator1
    Write Partition Index to Ribbon Link Order:ENiagara_AttributeSamplingApplyOutput=::NewEnumerator0
    Write Partition to Ribbon ID=::NewEnumerator0
   )
  ]
  particleUpdateScript [
   &ParticleState()
   &SkeletalMeshLocation(
    Bone / Socket Index=&ReturnExecIndex()
    Bone / Socket Sampling Mode=::NewEnumerator5
    MaskSpawnGroup=false
    Mesh Sampling Type=::NewEnumerator6
    Sampled Position Offset=&MakeVector(
     X=&FloatFromCurve(
      FloatCurve=DataInterface() {propertyValues {Curve {keys [<at 0 -8>,<at 0.01 0>,<at 1 0>]}}}
      CurveIndex=&ReturnNormalizedExecIndex()
     )
     Y=0
     Z=&FloatFromCurve(
      FloatCurve=DataInterface() {propertyValues {Curve {keys [
          {arriveTangent=-48.7882, interpMode=RCIM_Cubic, leaveTangent=-48.7882, tangentMode=RCTM_User, time=0, value=14.5623}
          {arriveTangent=-30.1558, interpMode=RCIM_Cubic, leaveTangent=-30.1558, time=0.1, value=7.0623}
          {arriveTangent=-7.847, interpMode=RCIM_Cubic, leaveTangent=-7.847, time=0.4, value=2.5}
          <at 1 0 RCIM_Cubic>
         ]}}}
      CurveIndex=&ReturnNormalizedExecIndex()
     )
    )
    Skeletal Mesh=DataInterface() {propertyValues {
      DefaultMesh=null
      ExcludeBoneName=root
      OptionalComponentClass=null
      PreviewMesh=/BRCosmetics/Gadgets/Assets/VinderTech_GliderChute/Glider_Stamina_Male_StandAlone/Meshes/Stamina_Male_StandAlone_Glider
      SkinningMode=SkinOnTheFly
      SoftSourceActor=null
      SourceMode=Default
      UvSetIndex=0
      WholeMeshLOD=-1
      bExcludeBone=false
      bReadDeformedGeometry=true
      bRequireCurrentFrameData=true
      ComponentTags []
      FilteredBones [JNT_spine_01,JNT_spine_02,JNT_spine_03,JNT_spine_04,JNT_spine_05,JNT_spine_06,JNT_spine_07,JNT_spine_08,JNT_spine_09,JNT_spine_10,JNT_spine_11,JNT_spine_12,JNT_spine_13,JNT_spine_14,JNT_spine_15,JNT_spine_16,JNT_spine_17,JNT_spine_18,JNT_spine_19,JNT_spine_20,JNT_spine_21,JNT_spine_22,JNT_spine_23,JNT_spine_24,JNT_spine_25,JNT_spine_26,JNT_spine_27,JNT_spine_28,JNT_spine_29,JNT_spine_30,JNT_spine_31,JNT_spine_32,JNT_spine_33,JNT_spine_34,JNT_spine_35,JNT_spine_36,JNT_spine_37,JNT_spine_38,JNT_spine_39,JNT_spine_40,JNT_spine_41,JNT_spine_42,JNT_spine_43,JNT_spine_44,JNT_spine_45]
      FilteredSockets [FX_JNT_spine_45 Tip]
      MeshUserParameter {parameter {name=None, typeDefHandle {registeredTypeIndex=1}, varData []}}
      SamplingRegions []
     }}
    Use Sampled Mesh Attribute as Ribbon ID=false
   )
   &ScaleRibbonWidth(Ribbon Width Scale=&Multiply_Float(
     A=&Multiply_Float(
      A=&FloatFromCurve(
       FloatCurve=DataInterface() {propertyValues {Curve {keys [
           <at 0 0.15>
           <at 0.1 0.3 RCIM_Cubic>
           {arriveTangent=-0.4567, interpMode=RCIM_Cubic, leaveTangent=1.4294, tangentMode=RCTM_Break, time=0.9, value=0.1375}
           {arriveTangent=-0.7966, interpMode=RCIM_Cubic, leaveTangent=-0.7966, tangentMode=RCTM_User, time=0.999, value=0.15}
           <at 1 0.1047>
          ]}}}
       CurveIndex=&ReturnNormalizedExecIndex()
      )
      B=&FloatFromCurve(FloatCurve=DataInterface() {propertyValues {Curve {keys [<at 0.5 1 RCIM_Cubic>,<at 0.7 0 RCIM_Cubic>]}}})
     )
     B=&FloatFromCurve(
      FloatCurve=DataInterface() {propertyValues {Curve {keys [
          {interpMode=RCIM_Cubic, leaveTangentWeight=0.0333, tangentMode=RCTM_User, time=0, value=1}
          {
           arriveTangent=-2836.3694
           arriveTangentWeight=0.936
           interpMode=RCIM_Cubic
           leaveTangent=-2836.3694
           leaveTangentWeight=0.936
           tangentMode=RCTM_User
           tangentWeightMode=RCTWM_WeightedBoth
           time=0.15
           value=0
          }
         ]}}}
      CurveIndex=&Lerp_Float(
       A=1
       B=0
       Alpha=&Add_Float(
        A=&Add_Float(
         A=&Multiply_Float(A=$Particles.NormalizedAge, B=4)
         B=&ReturnNormalizedExecIndex()
        )
        B=0
       )
      )
     )
    ))
   /SetVariables(Particles.RibonCameraOffset:NiagaraPosition=&Add_Vector(
     A=$Particles.SkeletalMeshLocation.SampledPosition
     B=&Multiply_VectorByFloat(
      Vector=&ReturnCameraVector(Camera Property=::NewEnumerator4, Coordinate Space=::Simulation, Origin Position=$Particles.Position)
      Float=-150
     )
    )) {isSetParameters=true, module=SetVariables_#5}
  ]
  rendererClasses [&NiagaraRibbonRendererProperties]
  renderers [{rendererClass=&NiagaraRibbonRendererProperties}]
 }
 Claws_Glow {
  emitterSpawnScript [/SetVariables(Emitter.Number Rib Particles=5) {isSetParameters=true, module=SetVariables_#6}]
  emitterUpdateScript [
   &EmitterState(
    Inactive Response=::NewEnumerator1
    Life Cycle Mode=::NewEnumerator1
    Loop Behavior=::NewEnumerator1
    Loop Delay=1.2
    Loop Duration=10
    UseLoopDelay=true
   )
   &SpawnBurst_Instantaneous(Spawn Count=80, Use Spawn Probability=false, UseLoopCountLimit=false) {module=SpawnBurst_Instantaneous016}
  ]
  particleSpawnScript [
   &InitializeRibbon(Color=$System.Outline Color, Lifetime=0.7, Ribbon Width=25, UsePositionOffset=false, Write Material Random=true)
   &PartitionParticles(
    Number Of Partitions=16
    Particles Per Partition:NiagaraInt32=$Emitter.Number Rib Particles
    Partition Mode:ENiagara_PartitionMode=::NewEnumerator1
    Write Partition Index to Ribbon Link Order:ENiagara_AttributeSamplingApplyOutput=::NewEnumerator0
    Write Partition to Ribbon ID=::NewEnumerator0
   )
  ]
  particleUpdateScript [
   &ParticleState()
   /Niagara/Modules/Debug/SpriteBasedLine() {enabled=false}
   &SkeletalMeshLocation(
    Bone / Socket Index=&ReturnExecIndex()
    Bone Sampling Mode=::NewEnumerator4
    MaskSpawnGroup=false
    Sampled Position Offset=&MakeCustomVectorFromBool(
     Bool=&SetBoolByIntComparison(A=&ModuloInt(A=&ReturnExecIndex(), B=5), B=4, Comparison Type=::NewEnumerator4)
     True Vector=&MakeCustomVectorFromBool(
      Bool=&SetBoolByFloatComparison(
       A=&SelectFloatFromArray(
        Direct Array Index=/Niagara/DynamicInputs/Divide/Divide_Int(A:NiagaraInt32=&ReturnExecIndex(), B:NiagaraInt32=5)
        Float Selection Array=DataInterface() {propertyValues {FloatData [1,1,1,1,0,0,0,0,1,1,1,1,0,0,0,0]}}
       )
       B=0.5
      )
      True Vector=<25 0 0>
      False Vector=<-25 0 0>
     )
     False Vector=<0 0 0>
    )
    Skeletal Mesh=DataInterface() {propertyValues {
      DefaultMesh=null
      ExcludeBoneName=root
      OptionalComponentClass=null
      PreviewMesh=&Stamina_Male_StandAlone_Glider
      SkinningMode=SkinOnTheFly
      SoftSourceActor=null
      SourceMode=Default
      UvSetIndex=0
      WholeMeshLOD=-1
      bExcludeBone=false
      bReadDeformedGeometry=true
      bRequireCurrentFrameData=true
      ComponentTags []
      FilteredBones [
       JNT_L_wrist
       JNT_L_thumb_01
       JNT_L_thumb_02
       JNT_L_thumb_03
       JNT_L_thumb_03
       JNT_L_wrist
       JNT_L_index_01
       JNT_L_index_02
       JNT_L_index_03
       JNT_L_index_03
       JNT_L_wrist
       JNT_L_middle_01
       JNT_L_middle_02
       JNT_L_middle_03
       JNT_L_middle_03
       JNT_L_wrist
       JNT_L_pinky_01
       JNT_L_pinky_02
       JNT_L_pinky_03
       JNT_L_pinky_03
       JNT_R_wrist
       JNT_R_thumb_01
       JNT_R_thumb_02
       JNT_R_thumb_03
       JNT_R_thumb_03
       JNT_R_wrist
       JNT_R_index_01
       JNT_R_index_02
       JNT_R_index_03
       JNT_R_index_03
       JNT_R_wrist
       JNT_R_middle_01
       JNT_R_middle_02
       JNT_R_middle_03
       JNT_R_middle_03
       JNT_R_wrist
       JNT_R_pinky_01
       JNT_R_pinky_02
       JNT_R_pinky_03
       JNT_R_pinky_03
       JNT_L_foot
       JNT_L_pinky_toe_01
       JNT_L_pinky_toe_02
       JNT_L_pinky_toe_03
       JNT_L_pinky_toe_03
       JNT_L_foot
       JNT_L_big_toe_01
       JNT_L_big_toe_02
       JNT_L_big_toe_03
       JNT_L_big_toe_03
       JNT_L_foot
       JNT_L_index_toe_01
       JNT_L_index_toe_02
       JNT_L_index_toe_03
       JNT_L_index_toe_03
       JNT_L_foot
       JNT_L_middle_toe_01
       JNT_L_middle_toe_02
       JNT_L_middle_toe_03
       JNT_L_middle_toe_03
       JNT_R_foot
       JNT_R_big_toe_01
       JNT_R_big_toe_02
       JNT_R_big_toe_03
       JNT_R_big_toe_03
       JNT_R_foot
       JNT_R_index_toe_01
       JNT_R_index_toe_02
       JNT_R_index_toe_03
       JNT_R_index_toe_03
       JNT_R_foot
       JNT_R_middle_toe_01
       JNT_R_middle_toe_02
       JNT_R_middle_toe_03
       JNT_R_middle_toe_03
       JNT_R_foot
       JNT_R_pinky_toe_01
       JNT_R_pinky_toe_02
       JNT_R_pinky_toe_03
       JNT_R_pinky_toe_03
      ]
      FilteredSockets []
      MeshUserParameter {parameter {name=None, typeDefHandle {registeredTypeIndex=1}, varData []}}
      SamplingRegions []
     }}
    Use Sampled Mesh Attribute as Ribbon ID=false
   )
   &ScaleRibbonWidth(Ribbon Width Scale=&Multiply_Float(
     A=&FloatFromCurve(
      FloatCurve=DataInterface() {propertyValues {Curve {keys [{
           arriveTangentWeight=0.6571
           interpMode=RCIM_Cubic
           leaveTangentWeight=0.6571
           tangentMode=RCTM_User
           tangentWeightMode=RCTWM_WeightedBoth
           time=0
           value=1
          }]}}}
      CurveIndex=$Particles.PartitionParticles.Normalized Index In Partition
      Scale Curve=1.2
     )
     B=&FloatFromCurve(FloatCurve=DataInterface() {propertyValues {Curve {keys [
          {interpMode=RCIM_Cubic, leaveTangentWeight=0.05, tangentMode=RCTM_User, time=0, value=0}
          {arriveTangentWeight=0.0772, interpMode=RCIM_Cubic, leaveTangentWeight=0.3603, tangentMode=RCTM_User, time=0.15, value=1}
          {arriveTangentWeight=0.3603, interpMode=RCIM_Cubic, leaveTangentWeight=0.0772, tangentMode=RCTM_User, time=0.85, value=1}
          {arriveTangentWeight=0.05, interpMode=RCIM_Cubic, tangentMode=RCTM_User, time=1, value=0}
         ]}}})
    ))
   /SetVariables(Particles.RibonCameraOffset:NiagaraPosition=&Add_Vector(
     A=Linked() {linkedVariable {name=Particles.Position, type=NiagaraPosition}}
     B=&Multiply_VectorByFloat(
      Vector=&ReturnCameraVector(Camera Property=::NewEnumerator4, Coordinate Space=::Simulation, Origin Position=$Particles.Position)
      Float=-25
     )
    )) {isSetParameters=true, module=SetVariables_#5}
  ]
  rendererClasses [&NiagaraSpriteRendererProperties, &NiagaraRibbonRendererProperties]
  renderers [
   {rendererClass=&NiagaraSpriteRendererProperties}
   {rendererClass=&NiagaraRibbonRendererProperties, rendererIndex=1}
  ]
 }
 Corona {
  emitterSpawnScript [/SetVariables(Emitter.SizingCurve=DataInterface() {propertyValues {Curve {keys [
        {arriveTangent=155.1764, interpMode=RCIM_Cubic, leaveTangent=155.1764, tangentMode=RCTM_User, time=0, value=0}
        {arriveTangentWeight=0.175, interpMode=RCIM_Cubic, leaveTangentWeight=0.1588, tangentMode=RCTM_User, time=0.0262, value=1.8822}
        <at 0.15 1>
        <at 0.75 1>
        <at 1 0>
       ]}}}) {isSetParameters=true, module=SetVariables_#7}]
  emitterUpdateScript [
   &EmitterState(
    Inactive Response=::NewEnumerator1
    Life Cycle Mode=::NewEnumerator1
    Loop Behavior=::NewEnumerator1
    Loop Delay=&Add_Float(A=$System.GlobalDelay, B=-0.05)
    Loop Duration=4
    UseLoopDelay=true
   )
   &SpawnBurst_Instantaneous(Spawn Time=0.225, Use Spawn Probability=false, UseLoopCountLimit=false)
  ]
  particleSpawnScript [
   &InitializeParticle(
    Color=<rgba 15 7.65 0.3 1>
    Lifetime=2.25
    Mesh Scale Mode=::NewEnumerator3
    Mesh Uniform Scale=4
    Sprite Rotation Mode=::NewEnumerator2
    Sprite Size Mode=::NewEnumerator3
    Uniform Sprite Size=&FN_FrontEndFloatModifier(Gameplay=300, Front End=200)
    UsePositionOffset=true
    Write Material Random=true
   )
   &CameraOffset(Camera Offset Amount=&FN_FrontEndFloatModifier(Gameplay=75, Front End=20))
  ]
  particleUpdateScript [
   &ParticleState()
   /SetVariables(
    Particles.Corona Color:LinearColor=&MultiplyLinearColor(
     Linear Color=<rgba 2 1.02 0.04 0.1>
     Scale Factor=&MakeFloat4FromVectorAndFloat(
      "Vector (RGB)"=&VectorFromFloat(Value=1)
      "Float (Alpha)"=&FloatFromCurve(FloatCurve=DataInterface() {propertyValues {Curve {keys [
           {arriveTangent=8.9568, interpMode=RCIM_Cubic, leaveTangent=8.9568, tangentMode=RCTM_User, time=0, value=0}
           {arriveTangentWeight=0.1341, interpMode=RCIM_Cubic, leaveTangentWeight=0.2413, tangentMode=RCTM_User, time=0.3, value=1}
           {arriveTangentWeight=0.1593, interpMode=RCIM_Cubic, leaveTangentWeight=0.0885, tangentMode=RCTM_User, time=0.75, value=1}
           {arriveTangentWeight=0.0833, interpMode=RCIM_Cubic, tangentMode=RCTM_User, time=1, value=0}
          ]}}})
     )
    )
    Particles.Corona Size:Vector2f=&Multiply_Vector2D_ByFloat(
     Float=&Lerp_Float(
      A=0.9
      B=1
      Alpha=&FloatFromCurve(FloatCurve=DataInterface() {propertyValues {Curve {keys [
           {arriveTangent=60.4226, interpMode=RCIM_Cubic, leaveTangent=60.4226, tangentMode=RCTM_User, time=0, value=0}
           {arriveTangentWeight=0.175, interpMode=RCIM_Cubic, leaveTangentWeight=0.1588, tangentMode=RCTM_User, time=0.05, value=1}
           {arriveTangentWeight=0.1341, interpMode=RCIM_Cubic, leaveTangentWeight=0.2413, tangentMode=RCTM_User, time=0.3, value=0.5313}
           {arriveTangentWeight=0.1593, interpMode=RCIM_Cubic, leaveTangentWeight=0.0885, tangentMode=RCTM_User, time=0.75, value=0.5313}
           {arriveTangentWeight=0.0833, interpMode=RCIM_Cubic, tangentMode=RCTM_User, time=1, value=0.3985}
          ]}}})
     )
     Vector2D=<800 800>
    )
   ) {isSetParameters=true, module=SetVariables_#8}
  ]
  rendererClasses [&NiagaraSpriteRendererProperties]
  renderers [{rendererClass=&NiagaraSpriteRendererProperties}]
 }
 Drops_Curvyshockwave {
  emitterUpdateScript [
   &EmitterState(
    Life Cycle Mode=::NewEnumerator1
    Loop Behavior=::NewEnumerator1
    Loop Delay=&Add_Float(A=$System.GlobalDelay, B=0.1)
    Loop Duration=1
    UseLoopDelay=true
   )
   &SpawnRate(SpawnRate=&FN_FrontEndFloatModifier(Gameplay=8, Front End=8), Use Spawn Probability=false)
  ]
  particleSpawnScript [
   &InitializeParticle(
    Color=<rgba 1 0.51 0.02 1>
    Lifetime=&RandomRangeFloat(Minimum=0.6, Maximum=1.35)
    Sprite Rotation Mode=::NewEnumerator2
    Sprite Size=<240 137>
    Sprite Size Mode=::NewEnumerator4
    Sprite UV Mode=::NewEnumerator1
    UsePositionOffset=true
    Write Material Random=true
   )
   &CameraOffset(Camera Offset Amount=&RandomRangeFloat(Minimum=25, Maximum=25))
   &Color()
  ]
  particleUpdateScript [
   &ParticleState()
   &DynamicMaterialParameters(
    Index 0 Param 1=&FloatFromCurve(FloatCurve=DataInterface() {propertyValues {Curve {keys [
         {arriveTangent=-1.0417, interpMode=RCIM_Cubic, leaveTangent=-1.0417, tangentMode=RCTM_User, time=0.35, value=1.5}
         {
          arriveTangent=-0.151
          arriveTangentWeight=0.5463
          interpMode=RCIM_Cubic
          leaveTangent=-0.151
          leaveTangentWeight=0.5463
          tangentMode=RCTM_User
          tangentWeightMode=RCTWM_WeightedBoth
          time=1
          value=0.5
         }
        ]}}})
    Index 0 Param 3=&RandomRangeFloat(Minimum=0, Maximum=0)
    Index 0 Param 4=&RandomRangeFloat(Minimum=0, Maximum=1)
    Index 1 Param 1=1
    Index 1 Param 2=1
    Index 3 Param 1=1
    Index 3 Param 2=1
    Index 3 Param 3=1
    Index 3 Param 4=1
    Module.Param0WriteEnabled=true
    Module.Param0WriteEnabled1=true
    Module.Param0WriteEnabled2=true
    Module.Param0WriteEnabled3=true
    Module.Param1WriteEnabled=true
    Module.Param1WriteEnabled1=true
    Module.Param1WriteEnabled2=true
    Module.Param1WriteEnabled3=true
    Module.Param2WriteEnabled=true
    Module.Param2WriteEnabled1=true
    Module.Param2WriteEnabled2=true
    Module.Param2WriteEnabled3=true
    Module.Param3WriteEnabled=true
    Module.Param3WriteEnabled1=true
    Module.Param3WriteEnabled2=true
    Module.Param3WriteEnabled3=true
    Write Parameter Index 1=true
    Write Parameter Index 2=true
    Write Parameter Index 3=true
   )
   &ScaleColor(
    Scale Alpha=&FloatFromCurve(
     FloatCurve=DataInterface() {propertyValues {Curve {keys [{arriveTangentWeight=0.3986, leaveTangentWeight=0.0498, time=0.85, value=1}, {arriveTangentWeight=0.0333, time=1, value=0}]}}}
     Scale Curve=&FN_FrontEndFloatModifier(Gameplay=1, Front End=1)
    )
    Scale RGB=&VectorFromFloat(Value=1)
    ScaleA=true
    ScaleRGB=true
   )
   &ScaleSpriteSize(Scale Factor=&Vector2DFromFloat(Value=&FloatFromCurve(
      FloatCurve=DataInterface() {propertyValues {Curve {keys [
          {arriveTangent=1.1628, interpMode=RCIM_Cubic, leaveTangent=1.1628, tangentMode=RCTM_User, time=0, value=0.1}
          {
           arriveTangent=0.3006
           arriveTangentWeight=0.6821
           interpMode=RCIM_Cubic
           leaveTangent=0.3006
           leaveTangentWeight=0.6821
           tangentMode=RCTM_User
           tangentWeightMode=RCTWM_WeightedBoth
           time=1
           value=1
          }
         ]}}}
      Scale Curve=&RandomRangeFloat(Minimum=1, Maximum=2.25)
     )))
   &SpriteRotationRate(Rotation Rate=&RandomRangeFloat(Minimum=50, Maximum=-50))
  ]
  rendererClasses [&NiagaraSpriteRendererProperties]
  renderers [{rendererClass=&NiagaraSpriteRendererProperties}]
 }
 Drops_Curvyshockwave001 {
  emitterUpdateScript [
   &EmitterState(
    Life Cycle Mode=::NewEnumerator1
    Loop Behavior=::NewEnumerator1
    Loop Delay=&Add_Float(A=$System.GlobalDelay, B=0.1)
    Loop Duration=1
    UseLoopDelay=true
   )
   &SpawnRate(SpawnRate=&FN_FrontEndFloatModifier(Gameplay=20, Front End=12), Use Spawn Probability=false)
  ]
  particleSpawnScript [
   &InitializeParticle(
    Color=<rgba 16 8.16 0.32 1>
    Lifetime=&RandomRangeFloat(Minimum=0.35, Maximum=0.6)
    Sprite Rotation Mode=::NewEnumerator2
    Sprite Size=<240 137>
    Sprite Size Mode=::NewEnumerator4
    Sprite UV Mode=::NewEnumerator1
    UsePositionOffset=true
    Write Material Random=true
   )
   &CameraOffset(Camera Offset Amount=&RandomRangeFloat(Minimum=50, Maximum=25))
  ]
  particleUpdateScript [
   &ParticleState()
   &DynamicMaterialParameters(
    Index 0 Param 1=&FloatFromCurve(FloatCurve=DataInterface() {propertyValues {Curve {keys [<at 0.65 1>,<at 1 0>]}}})
    Index 0 Param 3=&RandomRangeFloat(Minimum=0, Maximum=0)
    Index 0 Param 4=&RandomRangeFloat(Minimum=0, Maximum=1)
    Index 1 Param 1=1
    Index 1 Param 2=1
    Index 3 Param 1=1
    Index 3 Param 2=1
    Index 3 Param 3=1
    Index 3 Param 4=1
    Module.Param0WriteEnabled=true
    Module.Param0WriteEnabled1=true
    Module.Param0WriteEnabled2=true
    Module.Param0WriteEnabled3=true
    Module.Param1WriteEnabled=true
    Module.Param1WriteEnabled1=true
    Module.Param1WriteEnabled2=true
    Module.Param1WriteEnabled3=true
    Module.Param2WriteEnabled=true
    Module.Param2WriteEnabled1=true
    Module.Param2WriteEnabled2=true
    Module.Param2WriteEnabled3=true
    Module.Param3WriteEnabled=true
    Module.Param3WriteEnabled1=true
    Module.Param3WriteEnabled2=true
    Module.Param3WriteEnabled3=true
    Write Parameter Index 1=true
    Write Parameter Index 2=true
    Write Parameter Index 3=true
   ) {module=DynamicMaterialParameters001}
   &ScaleColor(
    Scale Alpha=&FloatFromCurve(
     FloatCurve=DataInterface() {propertyValues {Curve {keys [{arriveTangentWeight=0.3986, leaveTangentWeight=0.0498, time=0.85, value=1}, {arriveTangentWeight=0.0333, time=1, value=0}]}}}
     Scale Curve=&FN_FrontEndFloatModifier(Gameplay=1, Front End=0.25)
    )
    Scale RGB=&VectorFromFloat(Value=&FNTimeOfDayScalar(Night Scale=0.1, Morning Scale=1, Evening Scale=1))
    ScaleA=true
    ScaleRGB=true
   )
   &ScaleSpriteSize(Scale Factor=&Vector2DFromFloat(Value=&FloatFromCurve(
      FloatCurve=DataInterface() {propertyValues {Curve {keys [
          {arriveTangent=1.1628, interpMode=RCIM_Cubic, leaveTangent=1.1628, tangentMode=RCTM_User, time=0, value=0.1}
          {
           arriveTangent=0.3006
           arriveTangentWeight=0.6821
           interpMode=RCIM_Cubic
           leaveTangent=0.3006
           leaveTangentWeight=0.6821
           tangentMode=RCTM_User
           tangentWeightMode=RCTWM_WeightedBoth
           time=1
           value=1
          }
         ]}}}
      Scale Curve=&RandomRangeFloat(Minimum=0.5, Maximum=1)
     )))
   &SpriteRotationRate(Rotation Rate=&RandomRangeFloat(Minimum=-30, Maximum=30))
  ]
  rendererClasses [&NiagaraSpriteRendererProperties]
  renderers [{rendererClass=&NiagaraSpriteRendererProperties}]
 }
 Filler_Glow {
  emitterSpawnScript [/SetVariables(Emitter.Number Rib Particles=26) {isSetParameters=true, module=SetVariables_#6}]
  emitterUpdateScript [
   &EmitterState(
    Inactive Response=::NewEnumerator1
    Life Cycle Mode=::NewEnumerator1
    Loop Behavior=::NewEnumerator1
    Loop Delay=1.2
    Loop Duration=10
    UseLoopDelay=true
   )
   &SpawnBurst_Instantaneous(Spawn Count=5, Use Spawn Probability=false, UseLoopCountLimit=false)
  ]
  particleSpawnScript [
   &InitializeParticle(
    Color=$System.Outline Color
    Lifetime=0.7
    Sprite Size Mode=::NewEnumerator3
    Sprite UV Mode=::NewEnumerator3
    Uniform Sprite Size=&SelectFloatFromArray(
     Direct Array Index=&ReturnExecIndex()
     Float Selection Array=DataInterface() {propertyValues {FloatData [30,30,30,30,30]}}
    )
    UsePositionOffset=false
    Write Material Random=true
   )
   &PartitionParticles(
    Particles Per Partition:NiagaraInt32=$Emitter.Number Rib Particles
    Partition Mode:ENiagara_PartitionMode=::NewEnumerator1
    Write Partition Index to Ribbon Link Order:ENiagara_AttributeSamplingApplyOutput=::NewEnumerator0
   )
   &CameraOffset(Camera Offset Amount=-35)
  ]
  particleUpdateScript [
   &ParticleState()
   &SkeletalMeshLocation(
    Bone / Socket Index=&ReturnExecIndex()
    Bone / Socket Sampling Mode=::NewEnumerator5
    MaskSpawnGroup=false
    Mesh Sampling Type=::NewEnumerator6
    Skeletal Mesh=DataInterface() {propertyValues {
      DefaultMesh=null
      ExcludeBoneName=root
      OptionalComponentClass=null
      PreviewMesh=&Stamina_Male_StandAlone_Glider
      SkinningMode=SkinOnTheFly
      SoftSourceActor=null
      SourceMode=Default
      UvSetIndex=0
      WholeMeshLOD=-1
      bExcludeBone=false
      bReadDeformedGeometry=true
      bRequireCurrentFrameData=true
      ComponentTags []
      FilteredBones [JNT_L_wrist,JNT_R_wrist,JNT_L_foot,JNT_R_foot]
      FilteredSockets [FX_JNT_spine_45 Tip]
      MeshUserParameter {parameter {name=None, typeDefHandle {registeredTypeIndex=1}, varData []}}
      SamplingRegions []
     }}
    Use Sampled Mesh Attribute as Ribbon ID=false
   )
   &ScaleSpriteSize(
    Scale Sprite Size Mode=::NewEnumerator0
    Uniform Scale Factor=&FloatFromCurve(FloatCurve=DataInterface() {propertyValues {Curve {keys [
         {interpMode=RCIM_Cubic, leaveTangentWeight=0.05, tangentMode=RCTM_User, time=0, value=0}
         {arriveTangentWeight=0.0772, interpMode=RCIM_Cubic, leaveTangentWeight=0.3603, tangentMode=RCTM_User, time=0.15, value=1}
         {arriveTangentWeight=0.3603, interpMode=RCIM_Cubic, leaveTangentWeight=0.0772, tangentMode=RCTM_User, time=0.85, value=1}
         {arriveTangentWeight=0.05, interpMode=RCIM_Cubic, tangentMode=RCTM_User, time=1, value=0}
        ]}}})
   )
  ]
  rendererClasses [&NiagaraSpriteRendererProperties]
  renderers [{rendererClass=&NiagaraSpriteRendererProperties}]
 }
 Flare_Glow {
  emitterSpawnScript [/SetVariables(Emitter.SizingCurve=DataInterface() {propertyValues {Curve {keys [
        {arriveTangent=60.4226, interpMode=RCIM_Cubic, leaveTangent=60.4226, tangentMode=RCTM_User, time=0, value=0}
        {arriveTangentWeight=0.175, interpMode=RCIM_Cubic, leaveTangentWeight=0.1588, tangentMode=RCTM_User, time=0.05, value=1.8822}
        {arriveTangentWeight=0.1341, interpMode=RCIM_Cubic, leaveTangentWeight=0.2413, tangentMode=RCTM_User, time=0.2592, value=1}
        {arriveTangentWeight=0.1593, interpMode=RCIM_Cubic, leaveTangentWeight=0.0885, tangentMode=RCTM_User, time=0.4464, value=1}
        {arriveTangentWeight=0.0833, interpMode=RCIM_Cubic, leaveTangent=-7.6948, time=0.593, value=0}
       ]}}}) {isSetParameters=true, module=SetVariables_#7}]
  emitterUpdateScript [
   &EmitterState(
    Inactive Response=::NewEnumerator1
    Life Cycle Mode=::NewEnumerator1
    Loop Behavior=::NewEnumerator1
    Loop Delay=&Add_Float(A=$System.GlobalDelay, B=-0.05)
    Loop Duration=4
    UseLoopDelay=true
   )
   &SpawnBurst_Instantaneous(Spawn Time=0.225, Use Spawn Probability=false, UseLoopCountLimit=false)
  ]
  particleSpawnScript [
   &InitializeParticle(
    Color=&LerpLinearColors(
     StartColor=<rgba 15 5.69 0.3 1>
     EndColor=<rgba 7 2.6553 0.14 0.5>
     LerpFactor=&FN_FrontEndFloatModifier(Gameplay=0, Front End=1)
    )
    Lifetime=2
    Mesh Scale Mode=::NewEnumerator3
    Mesh Uniform Scale=&FN_FrontEndFloatModifier(Gameplay=4, Front End=3.2)
    Sprite Rotation Mode=::NewEnumerator2
    Sprite Size Mode=::NewEnumerator3
    Uniform Sprite Size=&FN_FrontEndFloatModifier(Gameplay=300, Front End=250)
    UsePositionOffset=true
    Write Material Random=true
   )
   &InitialMeshOrientation(Mesh Orientation Mode=::NewEnumerator7, Rotation=<0 -0.25 0>, Use Rotation Vector=true)
   &InitialMeshOrientation(
    Mesh Orientation Mode=::NewEnumerator7
    Rotation=&RandomRangeVector(Minimum=<0 0 0>, Maximum=<1 0 0>)
    Use Rotation Vector=true
   ) {module=InitialMeshOrientation001}
   &CameraOffset(Camera Offset Amount=&FN_FrontEndFloatModifier(Gameplay=75, Front End=100))
  ]
  particleUpdateScript [
   &ParticleState()
   &ScaleColor(
    Scale Alpha=&Multiply_Float(
     A=&Lerp_Float(A=0.75, B=1, Alpha=&Sine(Normalized Angle=$Emitter.Age, Period=0.5, Scale=0.5, Bias=0.5))
     B=&FloatFromCurve(FloatCurve=DataInterface() {propertyValues {Curve {keys [<at 0.95 1>,<at 1 0.1>]}}})
    )
    Scale RGB=&VectorFromFloat(Value=&Multiply_Float(
      A=&FloatFromCurve(FloatCurve=$Emitter.SizingCurve)
      B=&FNTimeOfDayScalar(Night Scale=0.5, Morning Scale=1, Evening Scale=1)
     ))
    ScaleA=true
    ScaleRGB=true
   ) {module=ScaleColor001}
   &ScaleMeshSize(Scale Factor=&VectorFromFloat(Value=&FloatFromCurve(FloatCurve=$Emitter.SizingCurve, Scale Curve=1.25)))
   &ScaleSpriteSize(Uniform Curve Sprite Scale:NiagaraDataInterfaceCurve=$Emitter.SizingCurve)
   /SetVariables(
    Particles.Corona Color:LinearColor=&MultiplyLinearColor(
     Linear Color=&LerpLinearColors(
      StartColor=<rgba 2 1.02 0.04 0.1>
      EndColor=<rgba 2 1.02 0.02 0.1>
      LerpFactor=&FN_FrontEndFloatModifier(Gameplay=0, Front End=1)
     )
     Scale Factor=&MakeFloat4FromVectorAndFloat(
      "Vector (RGB)"=&VectorFromFloat(Value=1)
      "Float (Alpha)"=&FloatFromCurve(FloatCurve=DataInterface() {propertyValues {Curve {keys [
           {arriveTangent=8.9568, interpMode=RCIM_Cubic, leaveTangent=8.9568, tangentMode=RCTM_User, time=0, value=0}
           {arriveTangentWeight=0.1341, interpMode=RCIM_Cubic, leaveTangentWeight=0.2413, tangentMode=RCTM_User, time=0.3, value=1}
           {arriveTangentWeight=0.1593, interpMode=RCIM_Cubic, leaveTangentWeight=0.0885, tangentMode=RCTM_User, time=0.75, value=1}
           {arriveTangentWeight=0.0833, interpMode=RCIM_Cubic, tangentMode=RCTM_User, time=1, value=0}
          ]}}})
     )
    )
    Particles.Corona Size:Vector2f=&Multiply_Vector2D_ByFloat(
     Float=&Lerp_Float(
      A=0.9
      B=1
      Alpha=&FN_FrontEndFloatModifier(
       Gameplay=&FloatFromCurve(FloatCurve=DataInterface() {propertyValues {Curve {keys [
            {arriveTangent=60.4226, interpMode=RCIM_Cubic, leaveTangent=60.4226, tangentMode=RCTM_User, time=0, value=0}
            {arriveTangentWeight=0.175, interpMode=RCIM_Cubic, leaveTangentWeight=0.1588, tangentMode=RCTM_User, time=0.05, value=1}
            {arriveTangentWeight=0.1341, interpMode=RCIM_Cubic, leaveTangentWeight=0.2413, tangentMode=RCTM_User, time=0.3, value=0.5313}
            {arriveTangentWeight=0.1593, interpMode=RCIM_Cubic, leaveTangentWeight=0.0885, tangentMode=RCTM_User, time=0.75, value=0.5313}
            {arriveTangentWeight=0.0833, interpMode=RCIM_Cubic, tangentMode=RCTM_User, time=1, value=0.3985}
           ]}}})
       Front End=&FloatFromCurve(FloatCurve=DataInterface() {propertyValues {Curve {keys [
            {arriveTangent=60.4226, interpMode=RCIM_Cubic, leaveTangent=60.4226, tangentMode=RCTM_User, time=0, value=0}
            {arriveTangentWeight=0.175, interpMode=RCIM_Cubic, leaveTangentWeight=0.1588, tangentMode=RCTM_User, time=0.05, value=1}
            {arriveTangentWeight=0.1341, interpMode=RCIM_Cubic, leaveTangentWeight=0.2413, tangentMode=RCTM_User, time=0.3, value=0.5313}
            {arriveTangentWeight=0.1593, interpMode=RCIM_Cubic, leaveTangentWeight=0.0885, tangentMode=RCTM_User, time=0.75, value=0.5313}
            {arriveTangentWeight=0.0833, interpMode=RCIM_Cubic, tangentMode=RCTM_User, time=1, value=0.3985}
           ]}}})
      )
     )
     Vector2D=<800 800>
    )
   ) {isSetParameters=true, module=SetVariables_#8}
   &FN_Light_Attributes(
    Color=$Particles.Color
    Create Color Attribute=true
    Create Exponent Attribute=true
    Create Radius Attribute=true
    Exponent=2
    Radius=300
    Scale Color=4
   )
  ]
  rendererClasses [
   &NiagaraMeshRendererProperties
   &NiagaraSpriteRendererProperties
   &NiagaraLightRendererProperties
  ]
  renderers [
   {rendererClass=&NiagaraMeshRendererProperties, rendererIndex=0}
   {rendererClass=&NiagaraSpriteRendererProperties, rendererIndex=1}
   {rendererClass=&NiagaraLightRendererProperties, rendererIndex=2}
  ]
 }
 Flare_GlowEnd {
  emitterUpdateScript [
   &EmitterState(
    Inactive Response=::NewEnumerator1
    Life Cycle Mode=::NewEnumerator1
    Loop Behavior=::NewEnumerator1
    Loop Delay=&Add_Float(A=$System.GlobalDelay, B=1.325)
    UseLoopDelay=true
   )
   &SpawnBurst_Instantaneous(Use Spawn Probability=false, UseLoopCountLimit=false)
  ]
  particleSpawnScript [
   &InitializeParticle(
    Color=<rgba 5 2.55 0.1 1>
    Lifetime=0.15
    Mesh Scale Mode=::NewEnumerator3
    Mesh Uniform Scale=4
    Sprite Rotation Mode=::NewEnumerator2
    Sprite Size Mode=::NewEnumerator3
    Uniform Sprite Size=300
    UsePositionOffset=true
    Write Material Random=true
   )
   &CameraOffset(Camera Offset Amount=75)
  ]
  particleUpdateScript [
   &ParticleState()
   &ScaleColor(
    Scale Alpha=&Multiply_Float(
     A=&Lerp_Float(A=0.75, B=1, Alpha=&Sine(Normalized Angle=$Emitter.Age, Period=0.5, Scale=0.5, Bias=0.5))
     B=&FloatFromCurve(FloatCurve=DataInterface() {propertyValues {Curve {keys [<at 0.95 1>,<at 1 0.1>]}}})
    )
    Scale RGB=&VectorFromFloat(Value=&FloatFromCurve(FloatCurve=DataInterface() {propertyValues {Curve {keys [<at 0.95 1>,<at 1 0>]}}}))
    ScaleA=true
    ScaleRGB=true
   ) {module=ScaleColor001}
   &ScaleSpriteSize(Uniform Curve Sprite Scale=DataInterface() {propertyValues {Curve {keys [
        {arriveTangent=4.7885, interpMode=RCIM_Cubic, leaveTangent=4.7885, tangentMode=RCTM_User, time=0, value=0}
        {
         arriveTangentWeight=0.2699
         interpMode=RCIM_Cubic
         leaveTangentWeight=0.1262
         tangentMode=RCTM_Break
         tangentWeightMode=RCTWM_WeightedBoth
         time=0.3
         value=1
        }
        {arriveTangent=-1.4286, interpMode=RCIM_Cubic, leaveTangent=-1.4286, time=0.6785, value=0.8}
        {arriveTangent=-2.0603, interpMode=RCIM_Cubic, leaveTangent=-2.0603, tangentMode=RCTM_User, time=1, value=0}
       ]}}})
  ]
  rendererClasses [&NiagaraSpriteRendererProperties]
  renderers [{rendererClass=&NiagaraSpriteRendererProperties}]
 }
 Head_Glow002 {
  emitterSpawnScript [/SetVariables(Emitter.Number Rib Particles=26) {isSetParameters=true, module=SetVariables_#6}]
  emitterUpdateScript [
   &EmitterState(Life Cycle Mode=::NewEnumerator1, Loop Behavior=::NewEnumerator1, Loop Delay=1.2, Loop Duration=10, UseLoopDelay=true)
   &SpawnBurst_Instantaneous(Spawn Count=8, Use Spawn Probability=false, UseLoopCountLimit=false)
   &SpawnBurst_Instantaneous(Use Spawn Probability=false, UseLoopCountLimit=false) {enabled=false, module=SpawnBurst_Instantaneous001}
  ]
  particleSpawnScript [
   &InitializeRibbon(Lifetime=10, Ribbon Width=16.0364, UsePositionOffset=false, Write Material Random=true) {enabled=false}
   &InitializeParticle(
    Color=$System.Outline Color
    Lifetime=0.8
    Sprite Size=&MakeVector2D(
     X=&SelectFloatFromArray(
      Direct Array Index=&ReturnExecIndex()
      Float Selection Array=DataInterface() {propertyValues {FloatData [65,65,80,80,60,40,40,60]}}
     )
     Y=0
    )
    Sprite Size Mode=::NewEnumerator4
    Sprite UV Mode=::NewEnumerator3
    UsePositionOffset=false
    Write Material Random=true
   )
   &PartitionParticles(
    Particles Per Partition:NiagaraInt32=$Emitter.Number Rib Particles
    Partition Mode:ENiagara_PartitionMode=::NewEnumerator1
    Write Partition Index to Ribbon Link Order:ENiagara_AttributeSamplingApplyOutput=::NewEnumerator0
   )
   /SetVariables(Particles.RibbonID=/Niagara/DynamicInputs/NiagaraID/MakeNiagaraIDFromInt(Index:NiagaraInt32=&SpawnGroup(), Acquire Tag:NiagaraInt32=0)) {isSetParameters=true, module=SetVariables_#9}
   &CameraOffset(Camera Offset Amount=-25)
  ]
  particleUpdateScript [
   &ParticleState()
   &SpriteBasedLine() {enabled=false}
   &SkeletalMeshLocation(
    Bone / Socket Index=&ReturnExecIndex()
    MaskSpawnGroup=false
    Mesh Sampling Type=::NewEnumerator5
    Skeletal Mesh=DataInterface() {propertyValues {
      DefaultMesh=null
      ExcludeBoneName=root
      OptionalComponentClass=null
      PreviewMesh=&Stamina_Male_StandAlone_Glider
      SkinningMode=SkinOnTheFly
      SoftSourceActor=null
      SourceMode=Default
      UvSetIndex=0
      WholeMeshLOD=-1
      bExcludeBone=false
      bReadDeformedGeometry=true
      bRequireCurrentFrameData=true
      ComponentTags []
      FilteredBones []
      FilteredSockets [FX_L_Horn,FX_R_Horn,FX_L_hair,FX_R_Hair,FX_Head,FX_L_Snout,FX_R_Snout,FX_Jaw]
      MeshUserParameter {parameter {name=None, typeDefHandle {registeredTypeIndex=1}, varData []}}
      SamplingRegions []
     }}
    Socket Sampling Mode=::NewEnumerator5
    Use Sampled Mesh Attribute as Ribbon ID=false
   )
   &ScaleRibbonWidth(Ribbon Width Scale=&FloatFromCurve(
     FloatCurve=DataInterface() {propertyValues {Curve {keys [<at 0 0.3 RCIM_Cubic>, {arriveTangent=-1.6725, interpMode=RCIM_Cubic, leaveTangent=-1.6725, tangentMode=RCTM_User, time=1, value=0}]}}}
     CurveIndex=&ReturnNormalizedExecIndex()
    )) {enabled=false}
   /SetVariables(Particles.RibonCameraOffset:NiagaraPosition=&Add_Vector(
     A=Linked() {linkedVariable {name=Particles.Position, type=NiagaraPosition}}
     B=&Multiply_VectorByFloat(
      Vector=&ReturnCameraVector(Camera Property=::NewEnumerator4, Coordinate Space=::Simulation, Origin Position=$Particles.Position)
      Float=-15
     )
    )) {enabled=false, isSetParameters=true, module=SetVariables_#5}
   &SpriteBasedLine(
    Point A:Vector3f=$Particles.SkeletalMeshLocation.SampledPosition
    Point B:NiagaraPosition=&Add_Vector(
     A=$Particles.SkeletalMeshLocation.SampledPosition
     B=&Multiply_VectorByFloat(
      Vector=&ReturnMeshOrientationAxis(Mesh Orientation Axis=::XAxis, Mesh Orientation Quat=$Particles.SkeletalMeshLocation.SampledRotation)
      Float=&SelectFloatFromArray(
       Direct Array Index=&ReturnExecIndex()
       Float Selection Array=DataInterface() {propertyValues {FloatData [130,130,120,120,100,90,90,110]}}
      )
     )
    )
    SpriteWidth:NiagaraFloat=&Multiply_Float(
     A=&MakeFloatFromVector2d(Vector 2D=$Particles.SpriteSize)
     B=&FloatFromCurve(FloatCurve=DataInterface() {propertyValues {Curve {keys [<at 0.85 1 RCIM_Cubic>,<at 1 0 RCIM_Cubic>]}}})
    )
   ) {module=SpriteBasedLine001}
   /SetVariables(Particles.LinePos:Vector3f=$Particles.SpriteBasedLine001.Position) {isSetParameters=true}
   /SetVariables(Particles.New Sprite Sized:Vector2f=&Multiply_Vector2D_ByFloat(
     Float=&FloatFromCurve(FloatCurve=DataInterface() {propertyValues {Curve {keys [
          {interpMode=RCIM_Cubic, leaveTangentWeight=0.0667, tangentMode=RCTM_User, time=0, value=0}
          {arriveTangentWeight=0.1029, interpMode=RCIM_Cubic, leaveTangentWeight=0.3345, tangentMode=RCTM_User, time=0.2, value=1}
          {arriveTangentWeight=0.3468, interpMode=RCIM_Cubic, leaveTangentWeight=0.08, tangentMode=RCTM_User, time=0.75, value=1}
          <at 1 0 RCIM_Cubic>
         ]}}})
     Vector2D=$Particles.SpriteBasedLine001.SpriteSize
    )) {isSetParameters=true}
  ]
  rendererClasses [&NiagaraSpriteRendererProperties, &NiagaraRibbonRendererProperties]
  renderers [
   {rendererClass=&NiagaraSpriteRendererProperties}
   {rendererClass=&NiagaraRibbonRendererProperties, rendererIndex=1}
  ]
 }
 Larg_Arc_Start {
  emitterUpdateScript [
   &EmitterState(
    Life Cycle Mode=::NewEnumerator1
    Loop Behavior=::NewEnumerator1
    Loop Delay=&Add_Float(A=$System.GlobalDelay, B=0.05)
    Loop Duration=1
    UseLoopDelay=true
   )
   &SpawnBurst_Instantaneous(Spawn Count=12, Use Spawn Probability=false, UseLoopCountLimit=false)
   /SetVariables(
    Emitter.Radom Height:NiagaraFloat=&RandomRangeFloat(Minimum=50, Maximum=75)
    Emitter.random Float:NiagaraFloat=&RandomRangeFloat(Minimum=0, Maximum=1)
   ) {isSetParameters=true, module=SetVariables_#10}
   /SetVariables(
    Emitter.ShapeX:NiagaraFloat=&RandomRangeFloat(Minimum=0, Maximum=1, Evaluation Type=::NewEnumerator1)
    Emitter.ShapeY:NiagaraFloat=&RandomRangeFloat(Minimum=0, Maximum=1, Evaluation Type=::NewEnumerator1)
   ) {isSetParameters=true, module=SetVariables_#11}
   &ShapeLocation(MaskSpawnGroup=false, Sphere Distribution=::NewEnumerator1, Sphere Radius=1200, U Position=0, UseOwnerScale=true, V Position=1)
  ]
  particleSpawnScript [
   &InitializeRibbon(Color=<rgba 0.5 0.255 0.01 1>, Lifetime=0.5, Ribbon Width=100, UsePositionOffset=true, Write Material Random=true)
   /SetVariables(
    Particles.RibbonID=&MakeNiagaraIDFromInt(Index:NiagaraInt32=$Emitter.LoopCount, Acquire Tag:NiagaraInt32=0)
    Particles.link order:NiagaraFloat=&MakeFloatFromInt(Int=&ReturnParticleIDAttributes())
   ) {isSetParameters=true}
  ]
  particleUpdateScript [
   &ParticleState()
   &ScaleColor(
    Scale Alpha=&FloatFromCurve(FloatCurve=DataInterface() {propertyValues {Curve {keys [<at 0 1>,<at 0.15 0.7>,<at 0.8 0>]}}})
    Scale RGB=&VectorFromFloat(Value=&Multiply_Float(
      A=&FloatFromCurve(FloatCurve=DataInterface() {propertyValues {Curve {keys [<at 0.2548 1>,<at 0.2684 75>,<at 0.8 1>]}}})
      B=&FNTimeOfDayScalar(Night Scale=0.5, Morning Scale=1, Evening Scale=1)
     ))
    ScaleA=true
    ScaleRGB=true
   )
   &ScaleRibbonWidth(Ribbon Width Scale=&Multiply_Float(
     A=&FloatFromCurve(FloatCurve=DataInterface() {propertyValues {Curve {keys [
          <at 0 0 RCIM_Cubic>
          {
           arriveTangent=6.6981
           arriveTangentWeight=0.5644
           interpMode=RCIM_Cubic
           leaveTangent=-2.7995
           leaveTangentWeight=0.7758
           tangentMode=RCTM_Break
           tangentWeightMode=RCTWM_WeightedBoth
           time=0.25
           value=1
          }
          {
           arriveTangent=-0.9491
           arriveTangentWeight=0.3986
           interpMode=RCIM_Cubic
           leaveTangent=-0.9491
           leaveTangentWeight=0.3986
           tangentMode=RCTM_User
           tangentWeightMode=RCTWM_WeightedBoth
           time=0.8
           value=0
          }
         ]}}})
     B=&FloatFromCurve(
      FloatCurve=DataInterface() {propertyValues {Curve {keys [
          {arriveTangent=3.731, interpMode=RCIM_Cubic, leaveTangent=3.731, tangentMode=RCTM_User, time=0, value=0}
          {
           arriveTangentWeight=0.1828
           interpMode=RCIM_Cubic
           leaveTangentWeight=0.4957
           tangentMode=RCTM_Break
           tangentWeightMode=RCTWM_WeightedBoth
           time=0.5
           value=1
          }
          {
           arriveTangentWeight=0.0667
           interpMode=RCIM_Cubic
           leaveTangent=-29.1585
           leaveTangentWeight=1.2457
           tangentMode=RCTM_Break
           tangentWeightMode=RCTWM_WeightedBoth
           time=0.9
           value=1
          }
          {
           arriveTangent=-2.3395
           arriveTangentWeight=0.174
           interpMode=RCIM_Cubic
           leaveTangent=-2.3395
           leaveTangentWeight=0.174
           tangentMode=RCTM_User
           tangentWeightMode=RCTWM_WeightedBoth
           time=1
           value=-0.2821
          }
         ]}}}
      CurveIndex=&ReturnNormalizedExecIndex()
     )
    )) {module=ScaleRibbonWidth001}
   /SetVariables(Particles.Scale Pos:Vector3f=&Add_Vector(
     A=&Lerp_Vector(
      A=$Emitter.ShapeLocation.ShapePosition
      B=/Niagara/DynamicInputs/Helpers/SimulationPosition()
      Alpha=&Multiply_Float(
       A=&FloatFromCurve(FloatCurve=DataInterface() {propertyValues {Curve {keys [
            {arriveTangent=6.1109, arriveTangentWeight=1.1704, leaveTangentWeight=1.1704, time=0, value=0}
            {arriveTangentWeight=0.2357, interpMode=RCIM_Cubic, leaveTangentWeight=0.2357, tangentMode=RCTM_User, time=0.2494, value=1}
           ]}}})
       B=&ReturnNormalizedExecIndex()
      )
     )
     B=&Multiply_VectorByFloat(
      Vector=/Niagara/DynamicInputs/Vectors/RotateVector(
       Vector To Rotate:Vector3f=&NormalizeVector(
        Vector To Normalize=&RandomRangeVector(Minimum=<-1 -1 0>, Maximum=<1 1 0>)
        Safe Normalize=true
        Safe Fallback Threshold=0
        Safe Fallback Vector=<0 0 0.001>
       )
       Yaw:NiagaraFloat=&Multiply_Float(A=$Particles.NormalizedAge, B=&RandomRangeFloat(Minimum=-3600, Maximum=3600))
       Pitch:NiagaraFloat=&Multiply_Float(A=$Particles.NormalizedAge, B=&RandomRangeFloat(Minimum=3600, Maximum=-3600))
       Roll:NiagaraFloat=&Multiply_Float(A=$Particles.NormalizedAge, B=&RandomRangeFloat(Minimum=-3600, Maximum=3600))
       Rotation Rate:NiagaraFloat=$Engine.DeltaTime
      )
      Float=&Multiply_Float(
       A=&FloatFromCurve(FloatCurve=DataInterface() {propertyValues {Curve {keys [<at 0.75 1>,<at 1 0>]}}}, Scale Curve=120)
       B=&OneMinusFloat(Float=&ReturnNormalizedExecIndex())
      )
     )
    )) {isSetParameters=true, module=SetVariables_#12}
  ]
  rendererClasses [&NiagaraRibbonRendererProperties]
  renderers [{rendererClass=&NiagaraRibbonRendererProperties}]
 }
 LargeArcs001 {
  emitterSpawnScript [/SetVariables(Emitter.num ribbon particles:NiagaraInt32=&MakeCustomIntFromBool(
     Bool=&IsPlatformSetActive(New Platform Set=DataInterface() {propertyValues {Platforms {qualityLevelMask=-1}}})
     True INT=5
     False INT=10
    )) {isSetParameters=true}]
  emitterUpdateScript [
   &EmitterState(
    Delay First Loop Only=true
    Life Cycle Mode=::NewEnumerator1
    Loop Behavior=::NewEnumerator2
    Loop Count:NiagaraInt32=&RandomRangeInt(Minimum=12, Maximum=15)
    Loop Delay=&Add_Float(A=0.125, B=$System.GlobalDelay)
    Loop Duration=0.075
    UseLoopDelay=true
   )
   &SpawnBurst_Instantaneous(Spawn Count=$Emitter.num ribbon particles, Use Spawn Probability=false, UseLoopCountLimit=false)
   /SetVariables(
    Emitter.Radom Height:NiagaraFloat=&RandomRangeFloat(Minimum=50, Maximum=75)
    Emitter.random Float:NiagaraFloat=&RandomRangeFloat(Minimum=0, Maximum=1)
   ) {isSetParameters=true, module=SetVariables_#10}
   /SetVariables(
    Emitter.ShapeX:NiagaraFloat=&RandomRangeFloat(Minimum=0, Maximum=1, Evaluation Type=::NewEnumerator1)
    Emitter.ShapeY:NiagaraFloat=&RandomRangeFloat(Minimum=0, Maximum=1, Evaluation Type=::NewEnumerator1)
   ) {isSetParameters=true, module=SetVariables_#11}
  ]
  particleSpawnScript [
   &InitializeRibbon(
    Color=&RandomRangeLinearColor(Minimum=<rgba 2 0.5 0 1>, Maximum=<rgba 1 0.961 0 1>)
    Lifetime=&UniformRangedFloat(Minimum=0.2, Maximum=0.35)
    Ribbon Width=42
    UsePositionOffset=true
    Write Material Random=true
   )
   &PartitionParticles(
    Particles Per Partition:NiagaraInt32=$Emitter.num ribbon particles
    Partition Count Mode:ENiagara_InfiniteLoopDuration=::NewEnumerator1
    Partition Mode:ENiagara_PartitionMode=::NewEnumerator1
    Write Partition Index to Ribbon Link Order:ENiagara_AttributeSamplingApplyOutput=::NewEnumerator0
    Write Partition to Ribbon ID=::NewEnumerator0
   )
   /SetVariables(Particles.PartitionParticles.Normalized Index In Partition:NiagaraFloat=&Multiply_Float(
     A=$Particles.PartitionParticles.Normalized Index In Partition
     B=&Divide_Float(
      A=1
      B=&OneMinusFloat(Float=&Divide_Float(A=1, B=&MakeFloatFromInt(Int=$Emitter.num ribbon particles)))
     )
    )) {isSetParameters=true}
   &ShapeLocation(
    MaskSpawnGroup=false
    Sphere Distribution=::NewEnumerator1
    Sphere Radius Position:NiagaraFloat=$Particles.PartitionParticles.Normalized Index In Partition
    U Position=$Emitter.ShapeX
    UseOwnerScale=true
    V Position=$Emitter.ShapeY
   )
  ]
  particleUpdateScript [
   &ParticleState()
   &ScaleColor(
    Scale Alpha=&FloatFromCurve(FloatCurve=DataInterface() {propertyValues {Curve {keys [
         {arriveTangent=8.0553, interpMode=RCIM_Cubic, leaveTangent=8.0553, tangentMode=RCTM_User, time=0, value=0}
         {arriveTangentWeight=0.1171, interpMode=RCIM_Cubic, leaveTangentWeight=0.3, tangentMode=RCTM_User, time=0.2113, value=1}
         {arriveTangentWeight=0.2913, interpMode=RCIM_Cubic, leaveTangentWeight=0.1333, tangentMode=RCTM_User, time=0.7524, value=1}
         {arriveTangent=-10.4882, interpMode=RCIM_Cubic, leaveTangent=-10.4882, tangentMode=RCTM_User, time=1, value=0}
        ]}}})
    Scale RGB=&VectorFromFloat(Value=0.35)
    ScaleA=true
    ScaleRGB=true
   )
   &ScaleRibbonWidth(Ribbon Width Scale=&Multiply_Float(
     A=&FloatFromCurve(FloatCurve=DataInterface() {propertyValues {Curve {keys [
          {interpMode=RCIM_Cubic, tangentMode=RCTM_User, time=0, value=0}
          {interpMode=RCIM_Cubic, tangentMode=RCTM_User, time=0.25, value=1}
          {arriveTangentWeight=0.1491, interpMode=RCIM_Cubic, leaveTangentWeight=0.2236, tangentMode=RCTM_User, time=0.75, value=1}
          {arriveTangentWeight=0.1, interpMode=RCIM_Cubic, tangentMode=RCTM_User, time=1, value=0}
         ]}}})
     B=&FloatFromCurve(
      FloatCurve=DataInterface() {propertyValues {Curve {keys [
          {interpMode=RCIM_Cubic, leaveTangentWeight=0.0667, tangentMode=RCTM_User, time=0, value=0}
          {arriveTangentWeight=0.1491, interpMode=RCIM_Cubic, leaveTangentWeight=0.2236, tangentMode=RCTM_User, time=0.1128, value=1}
          {arriveTangentWeight=0.1021, interpMode=RCIM_Cubic, leaveTangentWeight=0.1021, tangentMode=RCTM_User, time=0.5, value=1}
          {arriveTangentWeight=0.2236, interpMode=RCIM_Cubic, leaveTangentWeight=0.1491, tangentMode=RCTM_User, time=0.9515, value=1}
          {arriveTangentWeight=0.0667, interpMode=RCIM_Cubic, tangentMode=RCTM_User, time=1, value=0}
         ]}}}
      CurveIndex=$Particles.PartitionParticles.Normalized Index In Partition
     )
    )) {module=ScaleRibbonWidth001}
   &CurlNoiseForce(Noise Quality / Cost=::NewEnumerator0, Noise Strength=250, Pan Noise=false)
   &SolveForcesAndVelocity(Clamp Velocity=false, Limit Acceleration=false)
   &SphereLocation(
    MaskSpawnGroup=false
    Override Seed=false
    Sphere Origin:NiagaraPosition=$Particles.Position
    Sphere Radius=&Multiply_Float(
     A=$Particles.PartitionParticles.Normalized Index In Partition
     B=&RandomRangeFloat(Minimum=0, Maximum=5)
    )
    Surface Only=false
    UseNonUniformScale=false
    UseOffset=false
    UseSphereOrientation=false
   )
   /SetVariables(Particles.Scale Pos:Vector3f=&Lerp_Vector(
     A=&SimulationPosition()
     B=Linked() {linkedVariable {name=Particles.Position, type=NiagaraPosition}}
     Alpha=&Multiply_Float(A=4.25, B=$Particles.NormalizedAge)
    )) {isSetParameters=true, module=SetVariables_#12}
   &Color(Color=&RandomRangeLinearColor(Minimum=<rgba 2 0.5 0 1>, Maximum=<rgba 1 0.961 0 1>)) {enabled=false}
  ]
  rendererClasses [&NiagaraRibbonRendererProperties]
  renderers [{rendererClass=&NiagaraRibbonRendererProperties}]
 }
 LargeArcs002 {
  emitterSpawnScript [/SetVariables(Emitter.numTo SPawn:NiagaraInt32=&MakeCustomIntFromBool(
     Bool=&IsPlatformSetActive(New Platform Set=DataInterface() {propertyValues {Platforms {qualityLevelMask=-29}}})
     True INT=3
     False INT=5
    )) {isSetParameters=true}]
  emitterUpdateScript [
   &EmitterState(
    Delay First Loop Only=true
    Life Cycle Mode=::NewEnumerator1
    Loop Behavior=::NewEnumerator2
    Loop Count:NiagaraInt32=&RandomRangeInt(Minimum=4, Maximum=8)
    Loop Delay=0.75
    Loop Duration=0.1
    Recalculate Duration Each Loop=true
    UseLoopDelay=true
   )
   &SpawnBurst_Instantaneous(Spawn Count=$Emitter.numTo SPawn, Use Spawn Probability=false, UseLoopCountLimit=false)
   /SetVariables(Emitter.Bone Select:NiagaraInt32=&RandomRangeInt(Minimum=0, Maximum=41, Evaluation Type=::NewEnumerator1)) {isSetParameters=true}
   /SetVariables(Emitter.Offsett:Vector3f=&Multiply_VectorByFloat(
     Vector=&NormalizeVector(
      Vector To Normalize=&RandomRangeVector(Minimum=<2 -1 -1>, Maximum=<2 1 1>)
      Safe Normalize=true
      Safe Fallback Threshold=0
      Safe Fallback Vector=<0 0 0.001>
     )
     Float=&RandomRangeFloat(Minimum=100, Maximum=350)
    )) {isSetParameters=true}
   /SetVariables(Emitter.ribbon life:NiagaraFloat=&RandomRangeFloat(Minimum=0.2, Maximum=0.5, Evaluation Type=::NewEnumerator1)) {isSetParameters=true}
  ]
  particleSpawnScript [
   &InitializeRibbon(
    Color=<rgba 3 1.53 0.06 1>
    Lifetime=$Emitter.ribbon life
    Ribbon Width=&RandomRangeFloat(Minimum=50, Maximum=100)
    UsePositionOffset=false
    Write Material Random=true
   )
   /SetVariables(Particles.Get BoneSelectr:NiagaraInt32=$Emitter.Bone Select) {isSetParameters=true}
   /SetVariables(Particles.Get Offsett:Vector3f=$Emitter.Offsett) {isSetParameters=true}
   &PartitionParticles(
    Particles Per Partition:NiagaraInt32=$Emitter.numTo SPawn
    Partition Count Mode:ENiagara_InfiniteLoopDuration=::NewEnumerator1
    Partition Mode:ENiagara_PartitionMode=::NewEnumerator1
    Write Partition Index to Ribbon Link Order:ENiagara_AttributeSamplingApplyOutput=::NewEnumerator0
    Write Partition to Ribbon ID=::NewEnumerator0
   )
   /SetVariables(Particles.New Vector 4f:Vector4f=&MakeVector4(X=0, Y=0, Z=0, W=0)) {isSetParameters=true}
  ]
  particleUpdateScript [
   &ParticleState()
   &ScaleColor(
    Scale Alpha=&FloatFromCurve(FloatCurve=DataInterface() {propertyValues {Curve {keys [
         {arriveTangent=8.0553, interpMode=RCIM_Cubic, leaveTangent=8.0553, tangentMode=RCTM_User, time=0, value=0}
         {arriveTangentWeight=0.1171, interpMode=RCIM_Cubic, leaveTangentWeight=0.3, tangentMode=RCTM_User, time=0.2113, value=1}
         {arriveTangentWeight=0.2913, interpMode=RCIM_Cubic, leaveTangentWeight=0.1333, tangentMode=RCTM_User, time=0.7524, value=1}
         {arriveTangent=-10.4882, interpMode=RCIM_Cubic, leaveTangent=-10.4882, tangentMode=RCTM_User, time=1, value=0}
        ]}}})
    Scale RGB=&VectorFromFloat(Value=1)
    ScaleA=true
    ScaleRGB=true
   )
   &ScaleRibbonWidth(Ribbon Width Scale=&Multiply_Float(
     A=&FloatFromCurve(FloatCurve=DataInterface() {propertyValues {Curve {keys [
          {interpMode=RCIM_Cubic, tangentMode=RCTM_User, time=0, value=1}
          {interpMode=RCIM_Cubic, tangentMode=RCTM_User, time=0.25, value=1}
          {arriveTangentWeight=0.1491, interpMode=RCIM_Cubic, leaveTangentWeight=0.2236, tangentMode=RCTM_User, time=0.75, value=1}
          {arriveTangentWeight=0.1, interpMode=RCIM_Cubic, tangentMode=RCTM_User, time=1, value=1}
         ]}}})
     B=&FloatFromCurve(
      FloatCurve=DataInterface() {propertyValues {Curve {keys [
          {
           arriveTangentWeight=0.2012
           interpMode=RCIM_Cubic
           leaveTangentWeight=0.2012
           tangentMode=RCTM_User
           tangentWeightMode=RCTWM_WeightedBoth
           time=0.5
           value=1
          }
          {arriveTangent=-6.3941, interpMode=RCIM_Cubic, leaveTangent=-6.3941, tangentMode=RCTM_User, time=0.7, value=0}
         ]}}}
      CurveIndex=$Particles.PartitionParticles.Normalized Index In Partition
      Scale Curve=0.25
     )
    )) {module=ScaleRibbonWidth001}
   &SkeletalMeshLocation(
    Bone / Socket Index=$Particles.Get BoneSelectr
    Bone Sampling Mode=::NewEnumerator4
    MaskSpawnGroup=false
    Skeletal Mesh=DataInterface() {propertyValues {
      DefaultMesh=null
      ExcludeBoneName=root
      OptionalComponentClass=null
      PreviewMesh=&Stamina_Male_StandAlone_Glider
      SkinningMode=SkinOnTheFly
      SoftSourceActor=null
      SourceMode=Default
      UvSetIndex=0
      WholeMeshLOD=-1
      bExcludeBone=false
      bReadDeformedGeometry=true
      bRequireCurrentFrameData=true
      ComponentTags []
      FilteredBones [JNT_spine_04,JNT_spine_05,JNT_spine_06,JNT_spine_07,JNT_spine_08,JNT_spine_09,JNT_spine_10,JNT_spine_11,JNT_spine_12,JNT_spine_13,JNT_spine_14,JNT_spine_15,JNT_spine_16,JNT_spine_17,JNT_spine_18,JNT_spine_19,JNT_spine_20,JNT_spine_21,JNT_spine_22,JNT_spine_23,JNT_spine_24,JNT_spine_25,JNT_spine_26,JNT_spine_27,JNT_spine_28,JNT_spine_29,JNT_spine_30,JNT_spine_31,JNT_spine_32,JNT_spine_33,JNT_spine_34,JNT_spine_35,JNT_spine_36,JNT_spine_37,JNT_spine_38,JNT_spine_39,JNT_spine_40,JNT_spine_41,JNT_spine_42,JNT_spine_43,JNT_spine_44,JNT_spine_45]
      FilteredSockets []
      MeshUserParameter {parameter {name=None, typeDefHandle {registeredTypeIndex=1}, varData []}}
      SamplingRegions []
     }}
    Use Sampled Mesh Attribute as Ribbon ID=false
   )
   /SetVariables(Particles.Scale Position:NiagaraPosition=&Add_Vector(
     A=/Niagara/DynamicInputs/Lerp/Lerp_Position(
      A:NiagaraPosition=Linked() {linkedVariable {name=Particles.SkeletalMeshLocation.SampledPosition, type=Vector3f}}
      B:NiagaraPosition=&Add_Vector(
       A=&RotateVectorByQuaternion(Vector To Rotate=$Particles.Get Offsett, Quaternion=$Particles.SkeletalMeshLocation.SampledRotation)
       B=$Particles.SkeletalMeshLocation.SampledPosition
      )
      Alpha:NiagaraFloat=&Multiply_Float(A=$Particles.NormalizedAge, B=$Particles.PartitionParticles.Normalized Index In Partition)
     )
     B=&Multiply_Vector(
      A=&Multiply_VectorByFloat(
       Vector=&RandomVector(
        Vector Scale=&Lerp_Float(
         A=&MakeFloatFromVector(
          Vector=&RandomVector(
           Vector Scale=&RandomRangeFloat(Minimum=-25, Maximum=-100)
           Constrain to Hemisphere=true
           Hemisphere Constraint Vector=&ReturnMeshOrientationAxis(Mesh Orientation Axis=::XAxis, Mesh Orientation Quat=$Particles.SkeletalMeshLocation.SampledRotation)
          )
          Channel=::NewEnumerator0
         )
         B=&MakeFloatFromVector(
          Vector=&RandomVector(
           Vector Scale=&RandomRangeFloat(Minimum=50, Maximum=150)
           Constrain to Hemisphere=true
           Hemisphere Constraint Vector=&ReturnMeshOrientationAxis(Mesh Orientation Axis=::XAxis, Mesh Orientation Quat=$Particles.SkeletalMeshLocation.SampledRotation)
          )
          Channel=::NewEnumerator0
         )
         Alpha=$Particles.NormalizedAge
        )
        Constrain to Hemisphere=true
        Hemisphere Constraint Vector=&ReturnMeshOrientationAxis(Mesh Orientation Axis=::XAxis, Mesh Orientation Quat=$Particles.SkeletalMeshLocation.SampledRotation)
       )
       Float=&Sine(Normalized Angle=$Emitter.Age, Period=10, Scale=0.5, Bias=0.5)
      )
      B=&VectorFromFloat(Value=&FloatFromCurve(
        FloatCurve=DataInterface() {propertyValues {Curve {keys [{tangentMode=RCTM_User, time=0, value=0}, {tangentMode=RCTM_User, time=1, value=1}]}}}
        CurveIndex=$Particles.PartitionParticles.Normalized Index In Partition
       ))
     )
    )) {isSetParameters=true}
  ]
  rendererClasses [&NiagaraRibbonRendererProperties]
  renderers [{rendererClass=&NiagaraRibbonRendererProperties}]
 }
 Lightning_on_ring001 {
  emitterUpdateScript [
   &EmitterState()
   &SpawnBurst_Instantaneous(
    Spawn Count=&RandomRangeInt(Minimum=4, Maximum=6)
    Spawn Time=0.1
    Use Spawn Probability=false
    UseLoopCountLimit=false
   )
  ]
  particleSpawnScript [
   &InitializeParticle(
    Color=<rgba 75 38.25 1.5 1>
    Lifetime Max=0.18
    Lifetime Min=0.1
    Lifetime Mode=::NewEnumerator1
    Sprite Size Max=<50 50>
    Sprite Size Min=<25 15>
    Sprite Size Mode=::NewEnumerator2
    Sprite UV Mode=::NewEnumerator1
    UsePositionOffset=true
    Write Material Random=true
   )
   &DynamicMaterialParameters(
    Index 0 Param 1=&RandomRangeFloat(Minimum=1, Maximum=500)
    Index 0 Param 2=&RandomRangeFloat(Minimum=1, Maximum=2)
    Module.Param0WriteEnabled=true
    Module.Param1WriteEnabled=true
    Module.Param2WriteEnabled=true
    Module.Param3WriteEnabled=true
   )
   &AddVelocity(Velocity=&RandomVector(Vector Scale=50))
   &ShapeLocation(MaskSpawnGroup=false, Override Seed=false, Sphere Radius=0, UseOwnerScale=true)
  ]
  particleUpdateScript [
   &SolveForcesAndVelocity(Clamp Velocity=false, Limit Acceleration=false)
   &ParticleState()
   &ScaleSpriteSize(Uniform Curve Scale=2)
  ]
  rendererClasses [&NiagaraSpriteRendererProperties]
  renderers [{rendererClass=&NiagaraSpriteRendererProperties}]
 }
 Limbs_Glow {
  emitterSpawnScript [/SetVariables(Emitter.Number Rib Particles=4) {isSetParameters=true, module=SetVariables_#6}]
  emitterUpdateScript [
   &EmitterState(
    Inactive Response=::NewEnumerator1
    Life Cycle Mode=::NewEnumerator1
    Loop Behavior=::NewEnumerator1
    Loop Delay=1.2
    Loop Duration=10
    UseLoopDelay=true
   )
   &SpawnBurst_Instantaneous(Spawn Count=$Emitter.Number Rib Particles, Use Spawn Probability=false, UseLoopCountLimit=false)
   &SpawnBurst_Instantaneous(Spawn Count=$Emitter.Number Rib Particles, Spawn Group=1, Use Spawn Probability=false, UseLoopCountLimit=false) {module=SpawnBurst_Instantaneous001}
   &SpawnBurst_Instantaneous(Spawn Count=$Emitter.Number Rib Particles, Spawn Group=2, Use Spawn Probability=false, UseLoopCountLimit=false) {module=SpawnBurst_Instantaneous002}
   &SpawnBurst_Instantaneous(Spawn Count=$Emitter.Number Rib Particles, Spawn Group=3, Use Spawn Probability=false, UseLoopCountLimit=false) {module=SpawnBurst_Instantaneous003}
  ]
  particleSpawnScript [
   &InitializeRibbon(Color=$System.Outline Color, Lifetime=0.7, Ribbon Width=35, UsePositionOffset=false, Write Material Random=true)
   &PartitionParticles(
    Number Of Partitions=4
    Particles Per Partition:NiagaraInt32=$Emitter.Number Rib Particles
    Partition Mode:ENiagara_PartitionMode=::NewEnumerator1
    Write Partition Index to Ribbon Link Order:ENiagara_AttributeSamplingApplyOutput=::NewEnumerator0
   )
   /SetVariables(Particles.SpriteSize=<5 5>) {isSetParameters=true, module=SetVariables_#13}
   /SetVariables(Particles.RibbonID=&MakeNiagaraIDFromInt(Index:NiagaraInt32=&SpawnGroup(), Acquire Tag:NiagaraInt32=0)) {isSetParameters=true, module=SetVariables_#9}
  ]
  particleUpdateScript [
   &ParticleState()
   &SpriteBasedLine() {enabled=false}
   &SkeletalMeshLocation(
    Bone / Socket Index=&ReturnExecIndex()
    Bone Sampling Mode=::NewEnumerator4
    MaskSpawnGroup=false
    Skeletal Mesh=DataInterface() {propertyValues {
      DefaultMesh=null
      ExcludeBoneName=root
      OptionalComponentClass=null
      PreviewMesh=&Stamina_Male_StandAlone_Glider
      SkinningMode=SkinOnTheFly
      SoftSourceActor=null
      SourceMode=Default
      UvSetIndex=0
      WholeMeshLOD=-1
      bExcludeBone=false
      bReadDeformedGeometry=true
      bRequireCurrentFrameData=true
      ComponentTags []
      FilteredBones [
       JNT_spine_09
       JNT_L_shoulder
       JNT_L_elbow
       JNT_L_wrist
       JNT_spine_09
       JNT_R_shoulder
       JNT_R_elbow
       JNT_R_wrist
       JNT_spine_21
       JNT_L_thigh
       JNT_L_knee
       JNT_L_foot
       JNT_spine_21
       JNT_R_thigh
       JNT_R_knee
       JNT_R_foot
      ]
      FilteredSockets []
      MeshUserParameter {parameter {name=None, typeDefHandle {registeredTypeIndex=1}, varData []}}
      SamplingRegions []
     }}
    Use Sampled Mesh Attribute as Ribbon ID=false
   )
   &ScaleRibbonWidth(Ribbon Width Scale=&Multiply_Float(
     A=&FloatFromCurve(
      FloatCurve=DataInterface() {propertyValues {Curve {keys [
          <at 0 0 RCIM_Cubic>
          {arriveTangent=0.5, arriveTangentWeight=0.0833, interpMode=RCIM_Cubic, leaveTangent=0.5, leaveTangentWeight=0.3333, time=0.35, value=1.25}
          {interpMode=RCIM_Cubic, leaveTangent=-1.1089, time=1, value=0.5}
         ]}}}
      CurveIndex=$Particles.PartitionParticles.Normalized Index In Partition
      Scale Curve=1.25
     )
     B=&FloatFromCurve(FloatCurve=DataInterface() {propertyValues {Curve {keys [
          {interpMode=RCIM_Cubic, leaveTangentWeight=0.05, tangentMode=RCTM_User, time=0, value=0}
          {arriveTangentWeight=0.0772, interpMode=RCIM_Cubic, leaveTangentWeight=0.3603, tangentMode=RCTM_User, time=0.15, value=1}
          {arriveTangentWeight=0.3603, interpMode=RCIM_Cubic, leaveTangentWeight=0.0772, tangentMode=RCTM_User, time=0.85, value=1}
          {arriveTangentWeight=0.05, interpMode=RCIM_Cubic, tangentMode=RCTM_User, time=1, value=0}
         ]}}})
    ))
   /SetVariables(Particles.RibonCameraOffset:NiagaraPosition=&Add_Vector(
     A=Linked() {linkedVariable {name=Particles.Position, type=NiagaraPosition}}
     B=&Multiply_VectorByFloat(
      Vector=&ReturnCameraVector(Camera Property=::NewEnumerator4, Coordinate Space=::Simulation, Origin Position=$Particles.Position)
      Float=-100
     )
    )) {isSetParameters=true, module=SetVariables_#5}
  ]
  rendererClasses [&NiagaraSpriteRendererProperties, &NiagaraRibbonRendererProperties]
  renderers [
   {rendererClass=&NiagaraSpriteRendererProperties}
   {rendererClass=&NiagaraRibbonRendererProperties, rendererIndex=1}
  ]
 }
 MIN_Flash001 {
  emitterUpdateScript [
   &EmitterState(
    Life Cycle Mode=::NewEnumerator1
    Loop Behavior=::NewEnumerator1
    Loop Delay=&Add_Float(A=$System.GlobalDelay, B=0.15)
    Loop Duration=0.35
    UseLoopDelay=true
   )
   &SpawnRate(SpawnRate=12, Use Spawn Probability=false)
   &SpawnBurst_Instantaneous(
    Spawn Count=&RandomRangeInt(Minimum=4, Maximum=5)
    Use Spawn Probability=false
    UseLoopCountLimit=false
   )
  ]
  particleSpawnScript [
   &InitializeParticle(
    Color=&LerpLinearColors(
     StartColor=<rgba 25 25 25 1>
     EndColor=<rgba 3 3 3 1>
     LerpFactor=&FN_FrontEndFloatModifier(Gameplay=0, Front End=1)
    )
    Lifetime Max=0.65
    Lifetime Min=0.5
    Lifetime Mode=::NewEnumerator1
    Sprite Size=&Multiply_Vector2D_ByFloat(Float=7, Vector2D=&RandomRangeVector2D(Minimum=<3 35>, Maximum=<6 65>))
    Sprite Size Mode=::NewEnumerator4
    UsePositionOffset=false
    Write Material Random=true
   ) {module=InitializeParticle001}
   &SphereLocation(
    MaskSpawnGroup=false
    Override Seed=false
    Sphere Radius=1
    Surface Only=false
    UseNonUniformScale=false
    UseOffset=false
    UseSphereOrientation=false
   ) {module=SphereLocation001}
   &AddVelocity(Velocity=&Multiply_VectorByFloat(Vector=$Output.SphereLocation001.SphereNormal, Float=1))
  ]
  particleUpdateScript [
   &ParticleState()
   &SolveForcesAndVelocity(Clamp Velocity=false, Limit Acceleration=false)
   &ScaleColor(
    Scale Alpha=&FloatFromCurve(
     FloatCurve=DataInterface() {propertyValues {Curve {keys [{arriveTangent=-8.4754, interpMode=RCIM_Cubic, leaveTangent=-8.4754, tangentMode=RCTM_User, time=0.65, value=1}, <at 1 0 RCIM_Cubic>]}}}
     Scale Curve=&FNTimeOfDayScalar(Night Scale=0.25, Morning Scale=1, Evening Scale=1)
    )
    ScaleA=true
    ScaleRGB=true
   )
   &ScaleSpriteSize(
    Scale Factor=&MakeVector2D(
     X=&FloatFromCurve(
      FloatCurve=DataInterface() {propertyValues {Curve {keys [<at 0 1 RCIM_Cubic>, {arriveTangent=-0.2515, interpMode=RCIM_Cubic, leaveTangent=-0.2515, tangentMode=RCTM_User, time=1, value=0.5}]}}}
      Scale Curve=&FN_FrontEndFloatModifier(Gameplay=1, Front End=0.75)
     )
     Y=&FloatFromCurve(FloatCurve=DataInterface() {propertyValues {Curve {keys [{arriveTangent=1.6686, interpMode=RCIM_Cubic, leaveTangent=1.6686, tangentMode=RCTM_User, time=0, value=0}, <at 1 1 RCIM_Cubic>]}}})
    )
    Scale Sprite Size Mode=::NewEnumerator2
   )
  ]
  rendererClasses [&NiagaraSpriteRendererProperties]
  renderers [{rendererClass=&NiagaraSpriteRendererProperties}]
 }
 Orbs {
  emitterSpawnScript [/SetVariables(Emitter.GlowCurve=DataInterface() {propertyValues {Curve {keys [
        {interpMode=RCIM_Cubic, tangentMode=RCTM_User, time=0, value=1}
        {arriveTangentWeight=0.2457, interpMode=RCIM_Cubic, leaveTangentWeight=0.1966, tangentMode=RCTM_User, time=0.75, value=1}
        {arriveTangent=-5.9417, interpMode=RCIM_Cubic, leaveTangent=-5.9417, tangentMode=RCTM_User, time=1, value=0.025}
       ]}}}) {isSetParameters=true}]
  emitterUpdateScript [
   &EmitterState(Inactive Response=::NewEnumerator1, Life Cycle Mode=::NewEnumerator1, Loop Behavior=::NewEnumerator1, UseLoopDelay=false)
   &SpawnBurst_Instantaneous(Spawn Count=7, Use Spawn Probability=false, UseLoopCountLimit=false)
  ]
  particleSpawnScript [
   &InitializeParticle(
    Color=&FN_Select_Color_from_Time_Of_Day(Morning=<rgba 0.75 0.75 0.75 1>, Day=<rgba 1 1 1 1>, Evening=<rgba 0.75 0.75 0.75 1>, Night=<rgba 0.325 0.325 0.325 1>)
    Lifetime=1.25
    Sprite Size Mode=::NewEnumerator3
    Uniform Sprite Size=15
    UsePositionOffset=false
    Write Material Random=true
   )
   &DynamicMaterialParameters(
    Index 0 Param 1=&SelectFloatFromArray(
     Direct Array Index=&ReturnExecIndex()
     Float Selection Array=DataInterface() {propertyValues {FloatData [0,0.144,0.232,0.4,0.552,0.56,0.728]}}
    )
    Index 0 Param 2=1
    Module.Param0WriteEnabled=true
    Module.Param1WriteEnabled=true
    Module.Param2WriteEnabled=true
    Module.Param3WriteEnabled=true
   )
   &CameraOffset()
   &Color(Color=&MakeCustomLinearColorFromBool(
     Bool=&SetBoolByFloatComparison(A=&FN_FrontEndFloatModifier(Gameplay=0, Front End=1), B=0)
     True Color=<rgba 1 0 0 1>
     False Color=<rgba 0 1 0 1>
    )) {enabled=false}
  ]
  particleUpdateScript [
   &ParticleState()
   &SkeletalMeshLocation(
    Bone / Socket Index=&ReturnExecIndex()
    Bone Sampling Mode=::NewEnumerator4
    MaskSpawnGroup=false
    Skeletal Mesh=$User.SkeletalMesh
   )
   &self:OrbScale(
    Bone Index:NiagaraInt32=&ReturnExecIndex()
    New Skeletal Mesh:NiagaraDataInterfaceSkeletalMesh=$User.SkeletalMesh
   ) {module=OrbScale}
   /SetVariables(Particles.BoneScaleSize:Vector2f=&Multiply_Vector2D_ByFloat(
     Float=&FloatFromCurve(FloatCurve=DataInterface() {propertyValues {Curve {keys [<at 0.05 0>,<at 0.28 1>]}}})
     Vector2D=&Multiply_Vector2D_ByFloat(
      Float=&MakeFloatFromVector(Vector=$Particles.SampledBoneScale, Channel=::NewEnumerator0)
      Vector2D=$Particles.Initial.SpriteSize
     )
    )) {isSetParameters=true}
   &DynamicMaterialParameters(
    Index 0 Param 2=&FloatFromCurve(FloatCurve=$Emitter.GlowCurve)
    Module.Param0WriteEnabled=false
    Module.Param1WriteEnabled=true
    Module.Param2WriteEnabled=true
    Module.Param3WriteEnabled=true
   ) {module=DynamicMaterialParameters001}
   /SetVariables(Particles.Deploy Color:LinearColor=&LerpLinearColors(
     StartColor=<rgba 25 24.8251 0.8373 1>
     EndColor=$Particles.Color
     LerpFactor=&FloatFromCurve(FloatCurve=$Emitter.GlowCurve)
    )) {isSetParameters=true}
  ]
  rendererClasses [&NiagaraSpriteRendererProperties]
  renderers [{rendererClass=&NiagaraSpriteRendererProperties}]
 }
 ReF {
  bEnabled=false
  emitterUpdateScript [
   &EmitterState(Life Cycle Mode=::NewEnumerator1, Loop Duration=1, UseLoopDelay=false)
   &SpawnBurst_Instantaneous(Use Spawn Probability=false, UseLoopCountLimit=false)
  ]
  particleSpawnScript [&InitializeParticle(Color Mode=::NewEnumerator0, Lifetime=10, Sprite UV Mode=::NewEnumerator1, UsePositionOffset=false, Write Material Random=true)]
  particleUpdateScript [&ParticleState()]
  rendererClasses [&NiagaraComponentRendererProperties]
  renderers [{rendererClass=&NiagaraComponentRendererProperties, rendererIndex=0}]
 }
 Wiskers_Glow001 {
  emitterSpawnScript [/SetVariables(Emitter.Number Rib Particles=27) {isSetParameters=true, module=SetVariables_#6}]
  emitterUpdateScript [
   &EmitterState(
    Inactive Response=::NewEnumerator1
    Life Cycle Mode=::NewEnumerator1
    Loop Behavior=::NewEnumerator1
    Loop Delay=0.95
    Loop Duration=10
    UseLoopDelay=true
   )
   &SpawnBurst_Instantaneous(Spawn Count=$Emitter.Number Rib Particles, Use Spawn Probability=false, UseLoopCountLimit=false)
   &SpawnBurst_Instantaneous(Spawn Count=$Emitter.Number Rib Particles, Spawn Group=1, Use Spawn Probability=false, UseLoopCountLimit=false) {module=SpawnBurst_Instantaneous001}
  ]
  particleSpawnScript [
   &InitializeRibbon(Color=$System.Outline Color, Lifetime=1.5, Ribbon Width=22, UsePositionOffset=false, Write Material Random=true)
   &PartitionParticles(
    Particles Per Partition:NiagaraInt32=$Emitter.Number Rib Particles
    Partition Mode:ENiagara_PartitionMode=::NewEnumerator1
    Write Partition Index to Ribbon Link Order:ENiagara_AttributeSamplingApplyOutput=::NewEnumerator0
   )
   /SetVariables(Particles.SpriteSize=<5 5>) {isSetParameters=true, module=SetVariables_#13}
   /SetVariables(Particles.RibbonID=&MakeNiagaraIDFromInt(Index:NiagaraInt32=&SpawnGroup(), Acquire Tag:NiagaraInt32=0)) {isSetParameters=true, module=SetVariables_#9}
  ]
  particleUpdateScript [
   &ParticleState()
   &SpriteBasedLine() {enabled=false}
   &SkeletalMeshLocation(
    Bone / Socket Index=&SelectIntFromArray(
     Direct Array Index=&ReturnExecIndex()
     Int Selection Array=DataInterface() {propertyValues {IntData [52,0,1,2,3,4,5,6,7,8,9,10,11,12,13,14,15,16,17,18,19,20,21,22,23,24,25,52,26,27,28,29,30,31,32,33,34,35,36,37,38,39,40,41,42,43,44,45,46,47,48,49,50,51]}}
    )
    Bone / Socket Sampling Mode=::NewEnumerator5
    MaskSpawnGroup=false
    Mesh Sampling Type=::NewEnumerator6
    Skeletal Mesh=DataInterface() {propertyValues {
      DefaultMesh=null
      ExcludeBoneName=root
      OptionalComponentClass=null
      PreviewMesh=&Stamina_Male_StandAlone_Glider
      SkinningMode=SkinOnTheFly
      SoftSourceActor=null
      SourceMode=Default
      UvSetIndex=0
      WholeMeshLOD=-1
      bExcludeBone=false
      bReadDeformedGeometry=true
      bRequireCurrentFrameData=true
      ComponentTags []
      FilteredBones [
       JNT_R_whisker_01
       JNT_R_whisker_02
       JNT_R_whisker_03
       JNT_R_whisker_04
       JNT_R_whisker_05
       JNT_R_whisker_06
       JNT_R_whisker_07
       JNT_R_whisker_08
       JNT_R_whisker_09
       JNT_R_whisker_10
       JNT_R_whisker_11
       JNT_R_whisker_12
       JNT_R_whisker_13
       JNT_R_whisker_14
       JNT_R_whisker_15
       JNT_R_whisker_16
       JNT_R_whisker_17
       JNT_R_whisker_18
       JNT_R_whisker_19
       JNT_R_whisker_20
       JNT_R_whisker_21
       JNT_R_whisker_22
       JNT_R_whisker_23
       JNT_R_whisker_24
       JNT_R_whisker_25
       JNT_R_whisker_26
       JNT_L_whisker_01
       JNT_L_whisker_02
       JNT_L_whisker_03
       JNT_L_whisker_04
       JNT_L_whisker_05
       JNT_L_whisker_06
       JNT_L_whisker_07
       JNT_L_whisker_08
       JNT_L_whisker_09
       JNT_L_whisker_10
       JNT_L_whisker_11
       JNT_L_whisker_12
       JNT_L_whisker_13
       JNT_L_whisker_14
       JNT_L_whisker_15
       JNT_L_whisker_16
       JNT_L_whisker_17
       JNT_L_whisker_18
       JNT_L_whisker_19
       JNT_L_whisker_20
       JNT_L_whisker_21
       JNT_L_whisker_22
       JNT_L_whisker_23
       JNT_L_whisker_24
       JNT_L_whisker_25
       JNT_L_whisker_26
      ]
      FilteredSockets [FX_head_wiskerBase]
      MeshUserParameter {parameter {name=None, typeDefHandle {registeredTypeIndex=1}, varData []}}
      SamplingRegions []
     }}
    Use Sampled Mesh Attribute as Ribbon ID=false
   )
   &ScaleRibbonWidth(Ribbon Width Scale=&Multiply_Float(
     A=&Multiply_Float(
      A=&FloatFromCurve(
       FloatCurve=DataInterface() {propertyValues {Curve {keys [
           <at 0 0.3 RCIM_Cubic>
           {arriveTangent=-1.2357, interpMode=RCIM_Cubic, leaveTangent=-1.2357, tangentMode=RCTM_User, time=0.9818, value=0.0457}
          ]}}}
       CurveIndex=$Particles.PartitionParticles.Normalized Index In Partition
       Scale Curve=2
      )
      B=&FloatFromCurve(FloatCurve=DataInterface() {propertyValues {Curve {keys [<at 0.5165 1 RCIM_Cubic>,<at 0.6665 0 RCIM_Cubic>]}}})
     )
     B=&FloatFromCurve(
      FloatCurve=DataInterface() {propertyValues {Curve {keys [
          {interpMode=RCIM_Cubic, leaveTangentWeight=0.0333, tangentMode=RCTM_User, time=0, value=1}
          {
           arriveTangent=-2836.3694
           arriveTangentWeight=0.936
           interpMode=RCIM_Cubic
           leaveTangent=-2836.3694
           leaveTangentWeight=0.936
           tangentMode=RCTM_User
           tangentWeightMode=RCTWM_WeightedBoth
           time=0.1
           value=0
          }
         ]}}}
      CurveIndex=&Lerp_Float(
       A=1
       B=0
       Alpha=&Add_Float(
        A=&Add_Float(
         A=&Multiply_Float(A=$Particles.NormalizedAge, B=5.5)
         B=$Particles.PartitionParticles.Normalized Index In Partition
        )
        B=-0.5
       )
      )
     )
    ))
   /SetVariables(Particles.RibonCameraOffset:NiagaraPosition=&Add_Vector(
     A=Linked() {linkedVariable {name=Particles.Position, type=NiagaraPosition}}
     B=&Multiply_VectorByFloat(
      Vector=&ReturnCameraVector(Camera Property=::NewEnumerator4, Coordinate Space=::Simulation, Origin Position=$Particles.Position)
      Float=-15
     )
    )) {isSetParameters=true, module=SetVariables_#5}
  ]
  rendererClasses [&NiagaraSpriteRendererProperties, &NiagaraRibbonRendererProperties]
  renderers [
   {rendererClass=&NiagaraSpriteRendererProperties}
   {rendererClass=&NiagaraRibbonRendererProperties, rendererIndex=1}
  ]
 }
}
summary {
 systemName=NS_StaminaStandaloneGliderDeploy
 emitters [
  {bEnabled=false, emitterName=ReF, rendererClasses [&NiagaraComponentRendererProperties]}
  {emitterName=Body_Glow, rendererClasses [&NiagaraRibbonRendererProperties]}
  {
   emitterName=Wiskers_Glow001
   rendererClasses [&NiagaraSpriteRendererProperties, &NiagaraRibbonRendererProperties]
  }
  {emitterName=Limbs_Glow, rendererClasses [&NiagaraSpriteRendererProperties, &NiagaraRibbonRendererProperties]}
  {emitterName=Head_Glow002, rendererClasses [&NiagaraSpriteRendererProperties, &NiagaraRibbonRendererProperties]}
  {emitterName=Claws_Glow, rendererClasses [&NiagaraSpriteRendererProperties, &NiagaraRibbonRendererProperties]}
  {emitterName=LargeArcs002, rendererClasses [&NiagaraRibbonRendererProperties]}
  {emitterName=Filler_Glow, rendererClasses [&NiagaraSpriteRendererProperties]}
  {emitterName=Lightning_on_ring001, rendererClasses [&NiagaraSpriteRendererProperties]}
  {emitterName=LargeArcs001, rendererClasses [&NiagaraRibbonRendererProperties]}
  {emitterName=Drops_Curvyshockwave, rendererClasses [&NiagaraSpriteRendererProperties]}
  {
   emitterName=Flare_Glow
   rendererClasses [
    &NiagaraMeshRendererProperties
    &NiagaraSpriteRendererProperties
    &NiagaraLightRendererProperties
   ]
  }
  {emitterName=Larg_Arc_Start, rendererClasses [&NiagaraRibbonRendererProperties]}
  {emitterName=Drops_Curvyshockwave001, rendererClasses [&NiagaraSpriteRendererProperties]}
  {emitterName=Corona, rendererClasses [&NiagaraSpriteRendererProperties]}
  {emitterName=Flare_GlowEnd, rendererClasses [&NiagaraSpriteRendererProperties]}
  {emitterName=Orbs, rendererClasses [&NiagaraSpriteRendererProperties]}
  {emitterName=MIN_Flash001, rendererClasses [&NiagaraSpriteRendererProperties]}
 ]
 userVariables [
  User.bIsFrontEnd {type=NiagaraBool}
  User.bIsFrontEndPreview {type=NiagaraBool}
  User.RibbonIDOffset {type=NiagaraInt32}
  User.SkeletalMesh {
   type=NiagaraDataInterfaceSkeletalMesh
   defaultValue {
    struct=NiagaraExt_VariableValue_DataInterface
    value {dataInterface=&self:NiagaraDataInterfaceSkeletalMesh_0}
   }
  }
 ]
}
