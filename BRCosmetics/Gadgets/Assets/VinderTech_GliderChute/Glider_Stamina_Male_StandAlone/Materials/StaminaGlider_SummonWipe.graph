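# Material function that applies a glowing summon wipe to the stamina glider's material, driven by the SummonReveal parameter.
# Pre-skinned Y position, scaled by Size and mixed with TransitionNoise, is compared to a squared reveal threshold and rounded up into a hard opacity mask.
# An orange emissive color flares brighter near a midpoint of the reveal.
# The wipe attributes blend back to the input material late in the reveal.
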MaterialFunction /BRCosmetics/Gadgets/Assets/VinderTech_GliderChute/Glider_Stamina_Male_StandAlone/Materials/StaminaGlider_SummonWipe
graph {nodes [
  Add()
  n1 = GetMaterialAttributes(None=FunctionInput() {inputName=Material, inputType=FunctionInput_MaterialAttributes}) {attributeGetTypes [
    #1
    #2
    #3
    #4
   ]}
  n2 = Constant3Vector() {constant=<rgba 2 0.5501 0 0>}
  n3 = ScalarParameter() {defaultValue=0.96, parameterName=SummonReveal}
  n4 = LinearInterpolate(A=@n2, B=Multiply(A=@n2) {constB=3}, Alpha=SphereMask(A=@n3, B=Constant() {r=0.6}, Radius=Constant() {r=0.05}))
  n5 = VertexInterpolator(VS=PreSkinnedPosition())
  n6 = FunctionInput() {bUsePreviewValueAsDefault=true, inputName=TransitionNoise, inputType=FunctionInput_Scalar, previewValue=<1 0 0 1>}
  n7 = LinearInterpolate(
   A=Divide(
    A=MaterialFunctionCall(Float3=@n5.PS) {materialFunction=&BreakOutFloat3Components}.G
    B=ScalarParameter() {defaultValue=800, parameterName=Size}
   )
   B=@n6
  ) {constAlpha=0.2}
  n8 = LinearInterpolate(Alpha=@n3) {constA=0.2, constB=1.8}
  n9 = Ceil(None=Add(A=Subtract(A=@n7), B=Multiply(A=@n8, B=@n8)))
  n10 = LinearInterpolate(Alpha=@n3) {constA=-3, constB=1.66}
  FunctionOutput(None=BlendMaterialAttributes(
    A=SetMaterialAttributes!(MaterialAttributes=@n1.MaterialAttributes, EmissiveColor=@n4, OpacityMask=@n9) {attributeSetTypes [#5, #6]}
    B=@n1.MaterialAttributes
    Alpha=Saturate(None=@n10)
   ))
  LinearInterpolate()
  ObjectBounds()
  n11 = PreSkinnedLocalBounds()
  Subtract(A=@n11.Min, B=@n11.Min)
 ]}
